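# Physics asset for the female medium Bandit hoodie body cosmetic, set up for dangling bottom-sleeve pieces on both arms.
# Each bottom sleeve is a box body held by tight swing and twist limits to a tiny master-bone anchor capsule.
# Capsules on pelvis, lower spine, thighs, thigh twists and right hand form the body shapes, linked with angularly free constraints.
# The right side also has front, back and top sleeve anchors without child bodies.
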PhysicsAsset /BRCosmetics/Characters/Player/Female/Medium/Bodies/F_MED_Hoodie_Bandit/Meshes/F_MED_Hoodie_Bandit_Physics
bodies {
 dyn_sleeve_bk_01_r_MSTR [{shapeType=Capsule, length=0.01, radius=0.01}]
 dyn_sleeve_bt_01_l [{shapeType=Box, extentX=18, extentY=7.5, extentZ=2.5, center=<8.8795 -0.0002 0>}]
 dyn_sleeve_bt_01_l_MSTR [{shapeType=Capsule, length=0.01, radius=0.01}]
 dyn_sleeve_bt_01_r [{shapeType=Box, extentX=18, extentY=7.5, extentZ=2.5, center=<8.8795 -0.0002 0>}]
 dyn_sleeve_bt_01_r_MSTR [{shapeType=Capsule, length=0.01, radius=0.01}]
 dyn_sleeve_fr_01_r_MSTR [{shapeType=Capsule, length=0.01, radius=0.01}]
 dyn_sleeve_tp_01_r_MSTR [{shapeType=Capsule, length=0.01, radius=0.01}]
 hand_r [{shapeType=Capsule, length=5.72, radius=3.0925, center=<-3.6125 0.0456 0.3873>, rotation=<rot 6.7181 -89.9997 83.8401>}]
 pelvis [{shapeType=Capsule, length=13.0689, radius=10.7305, center=<-4.2112 -1.5235 0.0004>}]
 spine_01 [{shapeType=Capsule, length=10.75, radius=8.4805, center=<0 -1.0553 0>}]
 spine_02 [{shapeType=Capsule, length=9.25, radius=8.2305}]
 thigh_l [{shapeType=Capsule, length=41.0736, radius=8.0614, center=<-26.7089 0.0003 0.0007>, rotation=<rot 90 3.7927 3.7927>}]
 thigh_r [{shapeType=Capsule, length=41.0736, radius=8.0614, center=<27.4674 0 0.0007>, rotation=<rot 90 -2.5305 -2.5304>}]
 thigh_twist_01_l [{shapeType=Capsule, length=24.5736, radius=9.3114, center=<7.3561 0 0>, rotation=<rot 90 3.7927 3.7927>}]
 thigh_twist_01_r [{shapeType=Capsule, length=24.5736, radius=9.3114, center=<-7.3561 0 0>, rotation=<rot 90 3.7927 3.7927>}]
}
constraints [
 {bone1Name=dyn_sleeve_bt_01_r, bone2Name=dyn_sleeve_bt_01_r_MSTR, swing1LimitDegrees=15, swing2LimitDegrees=15, twistLimitDegrees=5}
 {
  bone1Name=hand_r
  bone2Name=spine_02
  swing1LimitDegrees=45
  swing1Motion=Free
  swing2LimitDegrees=45
  swing2Motion=Free
  twistLimitDegrees=45
  twistMotion=Free
 }
 {
  bone1Name=dyn_sleeve_bt_01_r_MSTR
  bone2Name=spine_02
  swing1LimitDegrees=45
  swing1Motion=Free
  swing2LimitDegrees=45
  swing2Motion=Free
  twistLimitDegrees=45
  twistMotion=Free
 }
 {
  bone1Name=dyn_sleeve_fr_01_r_MSTR
  bone2Name=spine_02
  swing1LimitDegrees=45
  swing1Motion=Free
  swing2LimitDegrees=45
  swing2Motion=Free
  twistLimitDegrees=45
  twistMotion=Free
 }
 {
  bone1Name=dyn_sleeve_bk_01_r_MSTR
  bone2Name=spine_02
  swing1LimitDegrees=45
  swing1Motion=Free
  swing2LimitDegrees=45
  swing2Motion=Free
  twistLimitDegrees=45
  twistMotion=Free
 }
 {
  bone1Name=dyn_sleeve_tp_01_r_MSTR
  bone2Name=spine_02
  swing1LimitDegrees=45
  swing1Motion=Free
  swing2LimitDegrees=45
  swing2Motion=Free
  twistLimitDegrees=45
  twistMotion=Free
 }
 {
  bone1Name=spine_01
  bone2Name=pelvis
  swing1LimitDegrees=45
  swing1Motion=Free
  swing2LimitDegrees=45
  swing2Motion=Free
  twistLimitDegrees=45
  twistMotion=Free
 }
 {
  bone1Name=spine_02
  bone2Name=spine_01
  swing1LimitDegrees=45
  swing1Motion=Free
  swing2LimitDegrees=45
  swing2Motion=Free
  twistLimitDegrees=45
  twistMotion=Free
 }
 {
  bone1Name=thigh_l
  bone2Name=pelvis
  swing1LimitDegrees=45
  swing1Motion=Free
  swing2LimitDegrees=45
  swing2Motion=Free
  twistLimitDegrees=45
  twistMotion=Free
 }
 {
  bone1Name=thigh_r
  bone2Name=pelvis
  swing1LimitDegrees=45
  swing1Motion=Free
  swing2LimitDegrees=45
  swing2Motion=Free
  twistLimitDegrees=45
  twistMotion=Free
 }
 {
  bone1Name=thigh_twist_01_r
  bone2Name=thigh_r
  swing1LimitDegrees=45
  swing1Motion=Free
  swing2LimitDegrees=45
  swing2Motion=Free
  twistLimitDegrees=45
  twistMotion=Free
 }
 {
  bone1Name=dyn_sleeve_bt_01_l_MSTR
  bone2Name=spine_02
  swing1LimitDegrees=45
  swing1Motion=Free
  swing2LimitDegrees=45
  swing2Motion=Free
  twistLimitDegrees=45
  twistMotion=Free
 }
 {bone1Name=dyn_sleeve_bt_01_l, bone2Name=dyn_sleeve_bt_01_l_MSTR, swing1LimitDegrees=15, swing2LimitDegrees=15, twistLimitDegrees=5}
 {
  bone1Name=thigh_twist_01_l
  bone2Name=thigh_l
  swing1LimitDegrees=45
  swing1Motion=Free
  swing2LimitDegrees=45
  swing2Motion=Free
  twistLimitDegrees=45
  twistMotion=Free
 }
]
