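# Post-pose AnimBP for the female Bandit hoodie body that adds secondary motion and corrective posing to hoodie, sleeves and boots.
# It copies the attached parent's pose, then chains bone constraints, crouch offsets, spring-driven AnimDynamics and sleeve Look Ats.
# The event graph feeds blend alphas and a gravity vector that follows skydiving force, or flips with pelvis rotation during emotes.
# Several branches, including RigidBody, are unconnected to the output.
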AnimBlueprint /BRCosmetics/Characters/Player/Female/Medium/Bodies/F_MED_Hoodie_Bandit/Meshes/F_MED_Hoodie_Bandit_AnimBP
target_skeleton=&/F_MED_Hoodie_Bandit_Skeleton
graphs [
 AnimGraph {nodes [
   n1 = Get windVector() -> (windVector:Vector)
   n2 = Get Aim_Off() -> (Aim_Off:double)
   n3 = Anim Dynamics - Bone: dyn_upperhoodie_01_l!(ComponentPose, Alpha=@n2, GravityOverride:Vector=@n1, ExternalForce="0,0,0") -> (Pose) {
    angularDampingOverride=1
    angularSpringConstant=1500
    bAngularSpring=true
    bGravityOverrideInSimSpace=true
    bOverrideAngularDamping=true
    bUseGravityOverride=true
    gravityScale=0.34
    simulationSpace=RootRelative
    alphaBoolBlend {blendOption=Linear}
    boundBone {boneName=dyn_upperhoodie_01_l}
    componentLinearAccScale=<0.3 0.3 0.5>
    componentLinearVelScale=<0.5 0.5 0.7>
    gravityOverride {z=-980}
    physicsBodyDefinitions [{
      sphereCollisionRadius=0
      boundBone {boneName=dyn_upperhoodie_01_l}
      boxExtents=<20 20 20>
      constraintSetup {angularLimitsMax=<0 0 15>, angularLimitsMin=<0 -15 -15>, angularTarget=<0 0 0>}
      localJointOffset=<8 0 0>
     }]
    relativeSpaceBone {boneName=spine_05}
   }
   Anim Dynamics - Bone: dyn_upperhoodie_fr_01_l!(ComponentPose=@n3, Alpha=@n2, GravityOverride:Vector=@n1, ExternalForce="0,0,0") -> (Pose) {
    angularDampingOverride=1
    angularSpringConstant=1500
    bAngularSpring=true
    bGravityOverrideInSimSpace=true
    bOverrideAngularDamping=true
    bUseGravityOverride=true
    gravityScale=0.34
    simulationSpace=RootRelative
    alphaBoolBlend {blendOption=Linear}
    boundBone {boneName=dyn_upperhoodie_fr_01_l}
    componentLinearAccScale=<0.3 0.3 0.5>
    componentLinearVelScale=<0.5 0.5 0.7>
    gravityOverride {z=-980}
    physicsBodyDefinitions [{
      sphereCollisionRadius=0
      boundBone {boneName=dyn_upperhoodie_fr_01_l}
      boxExtents=<20 20 20>
      constraintSetup {angularLimitsMax=<0 0 15>, angularLimitsMin=<0 -15 -15>, angularTarget=<0 0 0>}
      localJointOffset=<8 0 0>
     }]
    relativeSpaceBone {boneName=spine_05}
   }
   n4 = Anim Dynamics - Bone: dyn_upperhoodie_fr_01_r!(ComponentPose, Alpha=@n2, GravityOverride:Vector=@n1, ExternalForce="0,0,0") -> (Pose) {
    angularDampingOverride=1
    angularSpringConstant=1500
    bAngularSpring=true
    bGravityOverrideInSimSpace=true
    bOverrideAngularDamping=true
    bUseGravityOverride=true
    gravityScale=0.34
    simulationSpace=RootRelative
    alphaBoolBlend {blendOption=Linear}
    boundBone {boneName=dyn_upperhoodie_fr_01_r}
    componentLinearAccScale=<0.3 0.3 0.5>
    componentLinearVelScale=<0.5 0.5 0.7>
    gravityOverride {z=-980}
    physicsBodyDefinitions [{
      sphereCollisionRadius=0
      boundBone {boneName=dyn_upperhoodie_fr_01_r}
      boxExtents=<20 20 20>
      constraintSetup {angularLimitsMax=<0 0 15>, angularLimitsMin=<0 -15 -15>, angularTarget=<0 0 0>}
      localJointOffset=<8 0 0>
     }]
    relativeSpaceBone {boneName=spine_05}
   }
   Anim Dynamics - Bone: dyn_upperhoodie_01_r!(ComponentPose=@n4, Alpha=@n2, GravityOverride:Vector=@n1, ExternalForce="0,0,0") -> (Pose) {
    angularDampingOverride=1
    angularSpringConstant=1500
    bAngularSpring=true
    bGravityOverrideInSimSpace=true
    bOverrideAngularDamping=true
    bUseGravityOverride=true
    gravityScale=0.34
    simulationSpace=RootRelative
    alphaBoolBlend {blendOption=Linear}
    boundBone {boneName=dyn_upperhoodie_01_r}
    componentLinearAccScale=<0.3 0.3 0.5>
    componentLinearVelScale=<0.5 0.5 0.7>
    gravityOverride {z=-980}
    physicsBodyDefinitions [{
      sphereCollisionRadius=0
      boundBone {boneName=dyn_upperhoodie_01_r}
      boxExtents=<20 20 20>
      constraintSetup {angularLimitsMax=<0 0 15>, angularLimitsMin=<0 -15 -15>, angularTarget=<0 0 0>}
      localJointOffset=<8 0 0>
     }]
    relativeSpaceBone {boneName=spine_05}
   }
   n5 = Anim Dynamics - Bone: dyn_upperarm_hoodie_fr_00_r!(ComponentPose, Alpha=1, GravityOverride:Vector=@n1, ExternalForce="0,0,0") -> (Pose) {
    angularDampingOverride=1
    angularSpringConstant=1500
    bAngularSpring=true
    bGravityOverrideInSimSpace=true
    bOverrideAngularDamping=true
    bUseGravityOverride=true
    gravityScale=0.35
    simulationSpace=RootRelative
    alphaBoolBlend {blendOption=Linear}
    alphaScaleBiasClamp {scale=0.7}
    boundBone {boneName=dyn_upperarm_hoodie_fr_00_r}
    componentLinearAccScale=<0.3 0.3 0.5>
    componentLinearVelScale=<0.5 0.5 0.7>
    gravityOverride {z=-980}
    physicsBodyDefinitions [{
      sphereCollisionRadius=0
      boundBone {boneName=dyn_upperarm_hoodie_fr_00_r}
      boxExtents=<20 20 20>
      constraintSetup {angularLimitsMax=<5 20 0>, angularLimitsMin=<-5 -20 -20>, angularTarget=<0 0 0>}
      localJointOffset=<10 0 0>
     }]
   }
   Anim Dynamics - Bone: dyn_upperarm_hoodie_tp_00_r!(ComponentPose=@n5, Alpha=1, GravityOverride:Vector=@n1, ExternalForce="0,0,0") -> (Pose) {
    angularDampingOverride=1
    angularSpringConstant=1500
    bAngularSpring=true
    bGravityOverrideInSimSpace=true
    bOverrideAngularDamping=true
    bUseGravityOverride=true
    gravityScale=0.35
    simulationSpace=RootRelative
    alphaBoolBlend {blendOption=Linear}
    alphaScaleBiasClamp {scale=0.7}
    boundBone {boneName=dyn_upperarm_hoodie_tp_00_r}
    componentLinearAccScale=<0.3 0.3 0.5>
    componentLinearVelScale=<0.5 0.5 0.7>
    gravityOverride {z=-980}
    physicsBodyDefinitions [{
      sphereCollisionRadius=0
      boundBone {boneName=dyn_upperarm_hoodie_tp_00_r}
      boxExtents=<20 20 20>
      constraintSetup {angularLimitsMax=<5 0 20>, angularLimitsMin=<-5 -20 -20>, angularTarget=<0 0 0>}
      localJointOffset=<10 0 0>
     }]
   }
   n6 = Anim Dynamics - Bone: dyn_upperarm_hoodie_fr_00_l!(ComponentPose, Alpha=1, GravityOverride:Vector=@n1, ExternalForce="0,0,0") -> (Pose) {
    angularDampingOverride=1
    angularSpringConstant=1500
    bAngularSpring=true
    bGravityOverrideInSimSpace=true
    bOverrideAngularDamping=true
    bUseGravityOverride=true
    gravityScale=0.35
    simulationSpace=RootRelative
    alphaBoolBlend {blendOption=Linear}
    alphaScaleBiasClamp {scale=0.7}
    boundBone {boneName=dyn_upperarm_hoodie_fr_00_l}
    componentLinearAccScale=<0.3 0.3 0.5>
    componentLinearVelScale=<0.5 0.5 0.7>
    gravityOverride {z=-980}
    physicsBodyDefinitions [{
      sphereCollisionRadius=0
      boundBone {boneName=dyn_upperarm_hoodie_fr_00_l}
      boxExtents=<20 20 20>
      constraintSetup {angularLimitsMax=<5 0 20>, angularLimitsMin=<-5 -20 0>, angularTarget=<0 0 0>}
      localJointOffset=<10 0 0>
     }]
   }
   Anim Dynamics - Bone: dyn_upperarm_hoodie_tp_00_l!(ComponentPose=@n6, Alpha=1, GravityOverride:Vector=@n1, ExternalForce="0,0,0") -> (Pose) {
    angularDampingOverride=1
    angularSpringConstant=1500
    bAngularSpring=true
    bGravityOverrideInSimSpace=true
    bOverrideAngularDamping=true
    bUseGravityOverride=true
    gravityScale=0.35
    simulationSpace=RootRelative
    alphaBoolBlend {blendOption=Linear}
    alphaScaleBiasClamp {scale=0.7}
    boundBone {boneName=dyn_upperarm_hoodie_tp_00_l}
    componentLinearAccScale=<0.3 0.3 0.5>
    componentLinearVelScale=<0.5 0.5 0.7>
    gravityOverride {z=-980}
    physicsBodyDefinitions [{
      sphereCollisionRadius=0
      boundBone {boneName=dyn_upperarm_hoodie_tp_00_l}
      boxExtents=<20 20 20>
      constraintSetup {angularLimitsMax=<5 0 20>, angularLimitsMin=<-5 -20 -20>, angularTarget=<0 0 0>}
      localJointOffset=<10 0 0>
     }]
   }
   n7 = "Transform (Modify) Bone - Bone: dyn_lowerhoodie_bk_00"(Alpha=1, Translation="4.915,0.000,-0.000", Rotation="0,0,0", Scale="1,1,1") {
    translationMode=BMM_Additive
    translationSpace=BCS_BoneSpace
    alphaBoolBlend {blendOption=Linear}
    boneToModify {boneName=dyn_lowerhoodie_bk_00}
    translation {x=4.915}
   }
   n8 = "Transform (Modify) Bone - Bone: dyn_lowerhoodie_bk_00_r"(ComponentPose=@n7, Alpha=1, Translation="4.915,0.000,-0.000", Rotation="0,0,0", Scale="1,1,1") {
    translationMode=BMM_Additive
    translationSpace=BCS_BoneSpace
    alphaBoolBlend {blendOption=Linear}
    boneToModify {boneName=dyn_lowerhoodie_bk_00_r}
    translation {x=4.915}
   }
   n9 = "Transform (Modify) Bone - Bone: dyn_lowerhoodie_00_r"(ComponentPose=@n8, Alpha=1, Translation="4.915,0.000,-0.000", Rotation="0,0,0", Scale="1,1,1") {
    translationMode=BMM_Additive
    translationSpace=BCS_BoneSpace
    alphaBoolBlend {blendOption=Linear}
    boneToModify {boneName=dyn_lowerhoodie_00_r}
    translation {x=4.915}
   }
   n10 = "Transform (Modify) Bone - Bone: dyn_lowerhoodie_bk_00_l"(ComponentPose=@n9, Alpha=1, Translation="4.915,0.000,-0.000", Rotation="0,0,0", Scale="1,1,1") {
    translationMode=BMM_Additive
    translationSpace=BCS_BoneSpace
    alphaBoolBlend {blendOption=Linear}
    boneToModify {boneName=dyn_lowerhoodie_bk_00_l}
    translation {x=4.915}
   }
   "Transform (Modify) Bone - Bone: dyn_lowerhoodie_00_l"(ComponentPose=@n10, Alpha=1, Translation="4.915,0.000,-0.000", Rotation="0,0,0", Scale="1,1,1") {
    translationMode=BMM_Additive
    translationSpace=BCS_BoneSpace
    alphaBoolBlend {blendOption=Linear}
    boneToModify {boneName=dyn_lowerhoodie_00_l}
    translation {x=4.915}
   }
   n11 = Get handAlphaDown_l() -> (handAlphaDown_l:double)
   n12 = "Transform (Modify) Bone - Bone: dyn_sleeve_bk_02_r"(Alpha=@n11, Translation="-3,0,0", Rotation="0,0,0", Scale="1,1,1") {
    translationMode=BMM_Additive
    translationSpace=BCS_BoneSpace
    alphaBoolBlend {blendOption=Linear}
    boneToModify {boneName=dyn_sleeve_bk_02_r}
    translation {x=-3}
   }
   n13 = "Transform (Modify) Bone - Bone: dyn_sleeve_bt_02_r"(ComponentPose=@n12, Alpha=@n11, Translation="-3,0,0", Rotation="0,0,0", Scale="1,1,1") {
    translationMode=BMM_Additive
    translationSpace=BCS_BoneSpace
    alphaBoolBlend {blendOption=Linear}
    boneToModify {boneName=dyn_sleeve_bt_02_r}
    translation {x=-3}
   }
   n14 = "Transform (Modify) Bone - Bone: dyn_sleeve_bk_02_r"(ComponentPose=@n13, Alpha=@n11, Translation="-3,0,0", Rotation="0,0,0", Scale="1,1,1") {
    translationMode=BMM_Additive
    translationSpace=BCS_BoneSpace
    alphaBoolBlend {blendOption=Linear}
    boneToModify {boneName=dyn_sleeve_bk_02_r}
    translation {x=-3}
   }
   n15 = "Transform (Modify) Bone - Bone: dyn_sleeve_tp_02_r"(ComponentPose=@n14, Alpha=Get handAlphaDown_l_inv() -> (handAlphaDown_l_inv:double), Translation="-3,0,0", Rotation="0,0,0", Scale="1,1,1") {
    translationMode=BMM_Additive
    translationSpace=BCS_BoneSpace
    alphaBoolBlend {blendOption=Linear}
    boneToModify {boneName=dyn_sleeve_tp_02_r}
    translation {x=-3}
   }
   n16 = Get handAlphaDown_l()
   n17 = "Transform (Modify) Bone - Bone: dyn_sleeve_bk_02_l"(ComponentPose=@n15, Alpha=@n16, Translation="-3,0,0", Rotation="0,0,0", Scale="1,1,1") {
    translationMode=BMM_Additive
    translationSpace=BCS_BoneSpace
    alphaBoolBlend {blendOption=Linear}
    boneToModify {boneName=dyn_sleeve_bk_02_l}
    translation {x=-3}
   }
   n18 = "Transform (Modify) Bone - Bone: dyn_sleeve_bt_02_l"(ComponentPose=@n17, Alpha=@n16, Translation="-3,0,0", Rotation="0,0,0", Scale="1,1,1") {
    translationMode=BMM_Additive
    translationSpace=BCS_BoneSpace
    alphaBoolBlend {blendOption=Linear}
    boneToModify {boneName=dyn_sleeve_bt_02_l}
    translation {x=-3}
   }
   n19 = "Transform (Modify) Bone - Bone: dyn_sleeve_bk_02_l"(ComponentPose=@n18, Alpha=@n16, Translation="-3,0,0", Rotation="0,0,0", Scale="1,1,1") {
    translationMode=BMM_Additive
    translationSpace=BCS_BoneSpace
    alphaBoolBlend {blendOption=Linear}
    boneToModify {boneName=dyn_sleeve_bk_02_l}
    translation {x=-3}
   }
   n20 = Get handAlphaDown_l_inv()
   "Transform (Modify) Bone - Bone: dyn_sleeve_tp_02_l"(ComponentPose=@n19, Alpha=@n20, Translation="-3,0,0", Rotation="0,0,0", Scale="1,1,1") {
    translationMode=BMM_Additive
    translationSpace=BCS_BoneSpace
    alphaBoolBlend {blendOption=Linear}
    boneToModify {boneName=dyn_sleeve_tp_02_l}
    translation {x=-3}
   }
   "Transform (Modify) Bone - Bone: dyn_sleeve_tp_01_l"(Alpha=@n20, Translation="0,0,0", Rotation="40,0.000,0.000", Scale="1,1,1") {
    rotationMode=BMM_Additive
    rotationSpace=BCS_BoneSpace
    translationMode=BMM_Additive
    translationSpace=BCS_BoneSpace
    alphaBoolBlend {blendOption=Linear}
    boneToModify {boneName=dyn_sleeve_tp_01_l}
    rotation {pitch=40}
   }
   RigidBody!(ComponentPose, Alpha, OverrideWorldGravity:Vector=Get windVector(), ExternalForce) -> (Pose) {
    bClampLinearTranslationLimitToRefPose=true
    bOverrideWorldGravity=true
    overridePhysicsAsset=&/F_MED_Hoodie_Bandit_Physics
    simulationSpace=BaseBoneSpace
    alphaBoolBlend {blendOption=Linear}
    baseBoneRef {boneName=spine_01}
    componentAppliedLinearAccClamp=<2000 2000 2000>
    componentLinearAccScale=<0.3 0.3 0.3>
    componentLinearVelScale=<0.3 0.3 0.3>
    overrideWorldGravity {z=-980}
   }
   n21 = "Transform (Modify) Bone - Bone: dyn_lowerhoodie_fr_00_l_MSTR"(
    ComponentPose=Local To Component(LocalPose=Copy Pose From Mesh() {bUseAttachedParent=true})
    Alpha=Get isCrouchingAlpha() -> (isCrouchingAlpha:double)
    Translation="-10,0,0"
    Rotation="0,0,0"
    Scale="1,1,1"
   ) {
    translationMode=BMM_Additive
    translationSpace=BCS_BoneSpace
    alphaBoolBlend {blendOption=Linear}
    boneToModify {boneName=dyn_lowerhoodie_fr_00_l_MSTR}
    translation {x=-10}
   }
   n22 = "Constraint - dyn_lowerhoodie_fr_00_r_MSTR (R)"(ComponentPose=@n21, Alpha=1, ConstraintWeights_0=1) {
    alphaBoolBlend {blendOption=Linear}
    boneToModify {boneName=dyn_lowerhoodie_fr_00_r_MSTR}
    constraintSetup [{transformType=Rotation, targetBone {boneName=thigh_twist_01_r}}]
    constraintWeights [1]
   }
   n23 = Get isMobileLODAlpha() -> (isMobileLODAlpha:double)
   n24 = "Constraint - dyn_lowerhoodie_fr_00_l_MSTR (P)"(ComponentPose=@n22, Alpha=@n23, ConstraintWeights_0=1) {
    alphaBoolBlend {blendOption=Linear}
    boneToModify {boneName=dyn_lowerhoodie_fr_00_l_MSTR}
    constraintSetup [{targetBone {boneName=thigh_l}}]
    constraintWeights [1]
   }
   n25 = "Constraint - dyn_lowerhoodie_fr_00_r_MSTR (P)"(ComponentPose=@n24, Alpha=@n23, ConstraintWeights_0=1) {
    alphaBoolBlend {blendOption=Linear}
    boneToModify {boneName=dyn_lowerhoodie_fr_00_r_MSTR}
    constraintSetup [{targetBone {boneName=thigh_r}}]
    constraintWeights [1]
   }
   n26 = "Constraint - dyn_lowerhoodie_fr_00_MSTR (PP)"!(ComponentPose=@n25, Alpha=@n23, ConstraintWeights_0=0.5, ConstraintWeights_1:float=0.5) -> (Pose) {
    alphaBoolBlend {blendOption=Linear}
    boneToModify {boneName=dyn_lowerhoodie_fr_00_MSTR}
    constraintSetup [{targetBone {boneName=dyn_lowerhoodie_fr_00_l_MSTR}}, {targetBone {boneName=dyn_lowerhoodie_fr_00_MSTR}}]
    constraintWeights [1,1]
   }
   n27 = Get Aim_Off()
   n28 = Anim Dynamics - Bone: dyn_upperarm_hoodie_bt_00_r!(ComponentPose=@n26, Alpha=@n27, GravityOverride:Vector=@n1, ExternalForce="0,0,0") -> (Pose) {
    angularDampingOverride=1
    angularSpringConstant=1500
    bAngularSpring=true
    bGravityOverrideInSimSpace=true
    bOverrideAngularDamping=true
    bUseGravityOverride=true
    gravityScale=0.35
    simulationSpace=RootRelative
    alphaBoolBlend {blendOption=Linear}
    alphaScaleBiasClamp {bInterpResult=true, interpSpeedDecreasing=5, interpSpeedIncreasing=7, scale=0.7}
    boundBone {boneName=dyn_upperarm_hoodie_bt_00_r}
    componentLinearAccScale=<0.3 0.3 0.5>
    componentLinearVelScale=<0.5 0.5 0.7>
    gravityOverride {z=-980}
    physicsBodyDefinitions [{
      sphereCollisionRadius=0
      boundBone {boneName=dyn_upperarm_hoodie_bt_00_r}
      boxExtents=<20 20 20>
      constraintSetup {angularLimitsMax=<5 0 10>, angularLimitsMin=<-5 -20 -10>, angularTarget=<0 0 0>}
      localJointOffset=<10 0 0>
     }]
   }
   n29 = Anim Dynamics - Bone: dyn_upperarm_hoodie_bk_00_r!(ComponentPose=@n28, Alpha=@n27, GravityOverride:Vector=@n1, ExternalForce="0,0,0") -> (Pose) {
    angularDampingOverride=1
    angularSpringConstant=1500
    bAngularSpring=true
    bGravityOverrideInSimSpace=true
    bOverrideAngularDamping=true
    bUseGravityOverride=true
    gravityScale=0.35
    simulationSpace=RootRelative
    alphaBoolBlend {blendOption=Linear}
    alphaScaleBiasClamp {bInterpResult=true, interpSpeedDecreasing=5, interpSpeedIncreasing=7, scale=0.7}
    boundBone {boneName=dyn_upperarm_hoodie_bk_00_r}
    componentLinearAccScale=<0.3 0.3 0.5>
    componentLinearVelScale=<0.5 0.5 0.7>
    gravityOverride {z=-980}
    physicsBodyDefinitions [{
      sphereCollisionRadius=0
      boundBone {boneName=dyn_upperarm_hoodie_bk_00_r}
      boxExtents=<20 20 20>
      constraintSetup {angularLimitsMax=<5 0 20>, angularLimitsMin=<-5 0 0>, angularTarget=<0 0 0>}
      localJointOffset=<10 0 0>
     }]
   }
   n30 = Anim Dynamics - Bone: dyn_upperarm_hoodie_bt_00_l!(ComponentPose=@n29, Alpha=@n27, GravityOverride:Vector=@n1, ExternalForce="0,0,0") -> (Pose) {
    angularDampingOverride=1
    angularSpringConstant=1500
    bAngularSpring=true
    bGravityOverrideInSimSpace=true
    bOverrideAngularDamping=true
    bUseGravityOverride=true
    gravityScale=0.35
    simulationSpace=RootRelative
    alphaBoolBlend {blendOption=Linear}
    alphaScaleBiasClamp {bInterpResult=true, interpSpeedDecreasing=5, interpSpeedIncreasing=7, scale=0.7}
    boundBone {boneName=dyn_upperarm_hoodie_bt_00_l}
    componentLinearAccScale=<0.3 0.3 0.5>
    componentLinearVelScale=<0.5 0.5 0.7>
    gravityOverride {z=-980}
    physicsBodyDefinitions [{
      sphereCollisionRadius=0
      boundBone {boneName=dyn_upperarm_hoodie_bt_00_l}
      boxExtents=<20 20 20>
      constraintSetup {angularLimitsMax=<5 0 10>, angularLimitsMin=<-5 -20 -10>, angularTarget=<0 0 0>}
      localJointOffset=<10 0 0>
     }]
   }
   n31 = Anim Dynamics - Bone: dyn_upperarm_hoodie_bk_00_l!(ComponentPose=@n30, Alpha=@n27, GravityOverride:Vector=@n1, ExternalForce="0,0,0") -> (Pose) {
    angularDampingOverride=1
    angularSpringConstant=1500
    bAngularSpring=true
    bGravityOverrideInSimSpace=true
    bOverrideAngularDamping=true
    bUseGravityOverride=true
    gravityScale=0.35
    simulationSpace=RootRelative
    alphaBoolBlend {blendOption=Linear}
    alphaScaleBiasClamp {bInterpResult=true, interpSpeedDecreasing=5, interpSpeedIncreasing=7, scale=0.7}
    boundBone {boneName=dyn_upperarm_hoodie_bk_00_l}
    componentLinearAccScale=<0.3 0.3 0.5>
    componentLinearVelScale=<0.5 0.5 0.7>
    gravityOverride {z=-980}
    physicsBodyDefinitions [{
      sphereCollisionRadius=0
      boundBone {boneName=dyn_upperarm_hoodie_bk_00_l}
      boxExtents=<20 20 20>
      constraintSetup {angularLimitsMax=<5 0 0>, angularLimitsMin=<-5 0 -20>, angularTarget=<0 0 0>}
      localJointOffset=<10 0 0>
     }]
   }
   n32 = Anim Dynamics - Bone: dyn_lowerhoodie_bk_00_l!(ComponentPose=@n31, Alpha=@n2, GravityOverride:Vector=@n1, ExternalForce="0,0,0") -> (Pose) {
    angularDampingOverride=1
    angularSpringConstant=1500
    bAngularSpring=true
    bGravityOverrideInSimSpace=true
    bOverrideAngularDamping=true
    bUseGravityOverride=true
    gravityScale=0.34
    simulationSpace=RootRelative
    alphaBoolBlend {blendOption=Linear}
    alphaScaleBiasClamp {bInterpResult=true, interpSpeedDecreasing=5, interpSpeedIncreasing=7}
    boundBone {boneName=dyn_lowerhoodie_bk_00_l}
    componentLinearAccScale=<0.3 0.3 0.5>
    componentLinearVelScale=<0.5 0.5 0.7>
    gravityOverride {z=-980}
    physicsBodyDefinitions [{
      sphereCollisionRadius=0
      boundBone {boneName=dyn_lowerhoodie_bk_00_l}
      boxExtents=<20 20 20>
      constraintSetup {angularLimitsMax=<0 0 15>, angularLimitsMin=<0 -15 -15>, angularTarget=<0 0 0>}
      localJointOffset=<8 0 0>
     }]
    relativeSpaceBone {boneName=spine_04}
   }
   n33 = Anim Dynamics - Bone: dyn_lowerhoodie_fr_00_l!(ComponentPose=@n32, Alpha=@n2, GravityOverride:Vector=@n1, ExternalForce="0,0,0") -> (Pose) {
    angularDampingOverride=1
    angularSpringConstant=1500
    bAngularSpring=true
    bGravityOverrideInSimSpace=true
    bOverrideAngularDamping=true
    bUseGravityOverride=true
    gravityScale=0.34
    simulationSpace=RootRelative
    alphaBoolBlend {blendOption=Linear}
    alphaScaleBiasClamp {bInterpResult=true, interpSpeedDecreasing=5, interpSpeedIncreasing=7}
    boundBone {boneName=dyn_lowerhoodie_fr_00_l}
    componentLinearAccScale=<0.3 0.3 0.5>
    componentLinearVelScale=<0.5 0.5 0.7>
    gravityOverride {z=-980}
    physicsBodyDefinitions [{
      sphereCollisionRadius=0
      boundBone {boneName=dyn_lowerhoodie_fr_00_l}
      boxExtents=<20 20 20>
      constraintSetup {angularLimitsMax=<0 0 15>, angularLimitsMin=<0 -15 -15>, angularTarget=<0 0 0>}
      localJointOffset=<8 0 0>
     }]
    relativeSpaceBone {boneName=spine_05}
   }
   n34 = Anim Dynamics - Bone: dyn_lowerhoodie_00_l!(ComponentPose=@n33, Alpha=@n2, GravityOverride:Vector=@n1, ExternalForce="0,0,0") -> (Pose) {
    angularDampingOverride=1
    angularSpringConstant=1500
    bAngularSpring=true
    bGravityOverrideInSimSpace=true
    bOverrideAngularDamping=true
    bUseGravityOverride=true
    gravityScale=0.34
    simulationSpace=RootRelative
    alphaBoolBlend {blendOption=Linear}
    alphaScaleBiasClamp {bInterpResult=true, interpSpeedDecreasing=5, interpSpeedIncreasing=7}
    boundBone {boneName=dyn_lowerhoodie_00_l}
    componentLinearAccScale=<0.3 0.3 0.5>
    componentLinearVelScale=<0.5 0.5 0.7>
    gravityOverride {z=-980}
    physicsBodyDefinitions [{
      sphereCollisionRadius=0
      boundBone {boneName=dyn_lowerhoodie_00_l}
      boxExtents=<20 20 20>
      constraintSetup {angularLimitsMax=<0 0 15>, angularLimitsMin=<0 -15 -15>, angularTarget=<0 0 0>}
      localJointOffset=<8 0 0>
     }]
    relativeSpaceBone {boneName=spine_05}
   }
   n35 = Anim Dynamics - Bone: dyn_lowerhoodie_bk_00_r!(ComponentPose=@n34, Alpha=@n2, GravityOverride:Vector=@n1, ExternalForce="0,0,0") -> (Pose) {
    angularDampingOverride=1
    angularSpringConstant=1500
    bAngularSpring=true
    bGravityOverrideInSimSpace=true
    bOverrideAngularDamping=true
    bUseGravityOverride=true
    gravityScale=0.34
    simulationSpace=RootRelative
    alphaBoolBlend {blendOption=Linear}
    alphaScaleBiasClamp {bInterpResult=true, interpSpeedDecreasing=5, interpSpeedIncreasing=7}
    boundBone {boneName=dyn_lowerhoodie_bk_00_r}
    componentLinearAccScale=<0.3 0.3 0.5>
    componentLinearVelScale=<0.5 0.5 0.7>
    gravityOverride {z=-980}
    physicsBodyDefinitions [{
      sphereCollisionRadius=0
      boundBone {boneName=dyn_lowerhoodie_bk_00_r}
      boxExtents=<20 20 20>
      constraintSetup {angularLimitsMax=<0 0 15>, angularLimitsMin=<0 -15 -15>, angularTarget=<0 0 0>}
      localJointOffset=<8 0 0>
     }]
    relativeSpaceBone {boneName=spine_04}
   }
   n36 = Anim Dynamics - Bone: dyn_lowerhoodie_fr_00_r!(ComponentPose=@n35, Alpha=@n2, GravityOverride:Vector=@n1, ExternalForce="0,0,0") -> (Pose) {
    angularDampingOverride=1
    angularSpringConstant=1500
    bAngularSpring=true
    bGravityOverrideInSimSpace=true
    bOverrideAngularDamping=true
    bUseGravityOverride=true
    gravityScale=0.34
    simulationSpace=RootRelative
    alphaBoolBlend {blendOption=Linear}
    alphaScaleBiasClamp {bInterpResult=true, interpSpeedDecreasing=5, interpSpeedIncreasing=7}
    boundBone {boneName=dyn_lowerhoodie_fr_00_r}
    componentLinearAccScale=<0.3 0.3 0.5>
    componentLinearVelScale=<0.5 0.5 0.7>
    gravityOverride {z=-980}
    physicsBodyDefinitions [{
      sphereCollisionRadius=0
      boundBone {boneName=dyn_lowerhoodie_fr_00_r}
      boxExtents=<20 20 20>
      constraintSetup {angularLimitsMax=<0 0 15>, angularLimitsMin=<0 -15 -15>, angularTarget=<0 0 0>}
      localJointOffset=<8 0 0>
     }]
    relativeSpaceBone {boneName=spine_05}
   }
   n37 = Anim Dynamics - Bone: dyn_lowerhoodie_00_r!(ComponentPose=@n36, Alpha=@n2, GravityOverride:Vector=@n1, ExternalForce="0,0,0") -> (Pose) {
    angularDampingOverride=1
    angularSpringConstant=1500
    bAngularSpring=true
    bGravityOverrideInSimSpace=true
    bOverrideAngularDamping=true
    bUseGravityOverride=true
    gravityScale=0.34
    simulationSpace=RootRelative
    alphaBoolBlend {blendOption=Linear}
    alphaScaleBiasClamp {bInterpResult=true, interpSpeedDecreasing=5, interpSpeedIncreasing=7}
    boundBone {boneName=dyn_lowerhoodie_00_r}
    componentLinearAccScale=<0.3 0.3 0.5>
    componentLinearVelScale=<0.5 0.5 0.7>
    gravityOverride {z=-980}
    physicsBodyDefinitions [{
      sphereCollisionRadius=0
      boundBone {boneName=dyn_lowerhoodie_00_r}
      boxExtents=<20 20 20>
      constraintSetup {angularLimitsMax=<0 0 15>, angularLimitsMin=<0 -15 -15>, angularTarget=<0 0 0>}
      localJointOffset=<8 0 0>
     }]
    relativeSpaceBone {boneName=spine_05}
   }
   n38 = Anim Dynamics - Bone: dyn_lowerhoodie_bk_00!(ComponentPose=@n37, Alpha=@n2, GravityOverride:Vector="0,0,0", ExternalForce="0,0,0") -> (Pose) {
    angularDampingOverride=1
    angularSpringConstant=1500
    bAngularSpring=true
    bGravityOverrideInSimSpace=true
    bOverrideAngularDamping=true
    bUseGravityOverride=true
    gravityScale=0.34
    simulationSpace=RootRelative
    alphaBoolBlend {blendOption=Linear}
    alphaScaleBiasClamp {bInterpResult=true, interpSpeedDecreasing=5, interpSpeedIncreasing=7}
    boundBone {boneName=dyn_lowerhoodie_bk_00}
    componentLinearAccScale=<0.3 0.3 0.5>
    componentLinearVelScale=<0.5 0.5 0.7>
    gravityOverride {z=-980}
    physicsBodyDefinitions [{
      sphereCollisionRadius=0
      boundBone {boneName=dyn_lowerhoodie_bk_00}
      boxExtents=<20 20 20>
      constraintSetup {angularLimitsMax=<0 0 15>, angularLimitsMin=<0 -15 -15>, angularTarget=<0 0 0>}
      localJointOffset=<8 0 0>
     }]
    relativeSpaceBone {boneName=spine_05}
   }
   n39 = Anim Dynamics - Bone: dyn_lowerhoodie_fr_00!(ComponentPose=@n38, Alpha=@n2, GravityOverride:Vector="0,0,0", ExternalForce="0,0,0") -> (Pose) {
    angularDampingOverride=1
    angularSpringConstant=1500
    bAngularSpring=true
    bGravityOverrideInSimSpace=true
    bOverrideAngularDamping=true
    bUseGravityOverride=true
    gravityScale=0.34
    simulationSpace=RootRelative
    alphaBoolBlend {blendOption=Linear}
    alphaScaleBiasClamp {bInterpResult=true, interpSpeedDecreasing=5, interpSpeedIncreasing=7}
    boundBone {boneName=dyn_lowerhoodie_fr_00}
    componentLinearAccScale=<0.3 0.3 0.5>
    componentLinearVelScale=<0.5 0.5 0.7>
    gravityOverride {z=-980}
    physicsBodyDefinitions [{
      sphereCollisionRadius=0
      boundBone {boneName=dyn_lowerhoodie_fr_00}
      boxExtents=<20 20 20>
      constraintSetup {angularLimitsMax=<0 0 15>, angularLimitsMin=<0 -15 -15>, angularTarget=<0 0 0>}
      localJointOffset=<8 0 0>
     }]
    relativeSpaceBone {boneName=spine_05}
   }
   n40 = Anim Dynamics - Bone: dyn_spine01_l!(ComponentPose=@n39, Alpha=0.25, GravityOverride:Vector="0,0,0", ExternalForce="0,0,0") -> (Pose) {
    angularDampingOverride=1
    angularSpringConstant=1500
    bAngularSpring=true
    bGravityOverrideInSimSpace=true
    bOverrideAngularDamping=true
    bUseGravityOverride=true
    gravityScale=0.34
    simulationSpace=RootRelative
    alphaBoolBlend {blendOption=Linear}
    boundBone {boneName=dyn_spine01_l}
    componentLinearAccScale=<0.3 0.3 0.5>
    componentLinearVelScale=<0.5 0.5 0.7>
    gravityOverride {z=-980}
    physicsBodyDefinitions [{
      sphereCollisionRadius=0
      boundBone {boneName=dyn_spine01_l}
      boxExtents=<20 20 20>
      constraintSetup {angularLimitsMax=<0 0 15>, angularLimitsMin=<0 -15 -15>, angularTarget=<0 0 0>}
      localJointOffset=<8 0 0>
     }]
    relativeSpaceBone {boneName=spine_01}
   }
   n41 = Anim Dynamics - Bone: dyn_spine01_r!(ComponentPose=@n40, Alpha=0.25, GravityOverride:Vector="0,0,0", ExternalForce="0,0,0") -> (Pose) {
    angularDampingOverride=1
    angularSpringConstant=1500
    bAngularSpring=true
    bGravityOverrideInSimSpace=true
    bOverrideAngularDamping=true
    bUseGravityOverride=true
    gravityScale=0.34
    simulationSpace=RootRelative
    alphaBoolBlend {blendOption=Linear}
    boundBone {boneName=dyn_spine01_r}
    componentLinearAccScale=<0.3 0.3 0.5>
    componentLinearVelScale=<0.5 0.5 0.7>
    gravityOverride {z=-980}
    physicsBodyDefinitions [{
      sphereCollisionRadius=0
      boundBone {boneName=dyn_spine01_r}
      boxExtents=<20 20 20>
      constraintSetup {angularLimitsMax=<0 0 15>, angularLimitsMin=<0 -15 -15>, angularTarget=<0 0 0>}
      localJointOffset=<8 0 0>
     }]
    relativeSpaceBone {boneName=spine_01}
   }
   n42 = Look At - Bone: dyn_sleeve_tp_01_r_MSTR(ComponentPose=@n41, Alpha=Get handAlphaDown_r_inv() -> (handAlphaDown_r_inv:double)) {
    alphaBoolBlend {blendOption=Linear}
    boneToModify {boneName=dyn_sleeve_tp_01_r_MSTR}
    lookAtLocation {x=0}
    lookAtTarget {boneReference {boneName=dyn_sleeve_tp_02_r_aim}}
    lookAt_Axis {axis=<1 0>}
    lookUp_Axis {axis {x=0, z=1}}
   }
   n43 = Get handAlphaDown_r() -> (handAlphaDown_r:double)
   n44 = "Transform (Modify) Bone - Bone: dyn_sleeve_bk_02_r_aim"(ComponentPose=@n42, Alpha=@n43, Translation="0.000,0.000,7", Rotation="0,0,0", Scale="1,1,1") {
    translationMode=BMM_Additive
    translationSpace=BCS_BoneSpace
    alphaBoolBlend {blendOption=Linear}
    boneToModify {boneName=dyn_sleeve_bk_02_r_aim}
    translation {z=7}
   }
   n45 = Look At - Bone: dyn_sleeve_bk_01_r_MSTR(ComponentPose=@n44, Alpha=@n43) {
    alphaBoolBlend {blendOption=Linear}
    boneToModify {boneName=dyn_sleeve_bk_01_r_MSTR}
    lookAtLocation {x=0}
    lookAtTarget {boneReference {boneName=dyn_sleeve_bk_02_r_aim}}
    lookAt_Axis {axis=<1 0>}
    lookUp_Axis {axis {x=0, z=1}}
   }
   n46 = Look At - Bone: dyn_sleeve_bt_01_r_MSTR(ComponentPose=@n45, Alpha=1) {
    alphaBoolBlend {blendOption=Linear}
    boneToModify {boneName=dyn_sleeve_bt_01_r_MSTR}
    lookAtLocation {x=0}
    lookAtTarget {boneReference {boneName=dyn_sleeve_bt_02_r_aim}}
    lookAt_Axis {axis=<1 0>}
    lookUp_Axis {axis {x=0, z=1}}
   }
   n47 = Look At - Bone: dyn_sleeve_fr_01_r_MSTR(ComponentPose=@n46, Alpha=Get handAlphaDown_r_inv()) {
    alphaBoolBlend {blendOption=Linear}
    boneToModify {boneName=dyn_sleeve_fr_01_r_MSTR}
    lookAtLocation {x=0}
    lookAtTarget {boneReference {boneName=dyn_sleeve_fr_02_r_aim}}
    lookAt_Axis {axis=<1 0>}
    lookUp_Axis {axis {x=0, z=1}}
   }
   n48 = "Transform (Modify) Bone - Bone: dyn_sleeve_tp_02_l_aim"(ComponentPose=@n47, Alpha=1, Translation="0.000,-0.000,3.273", Rotation="0,0,0", Scale="1,1,1") {
    translationMode=BMM_Additive
    translationSpace=BCS_BoneSpace
    alphaBoolBlend {blendOption=Linear}
    boneToModify {boneName=dyn_sleeve_tp_02_l_aim}
    translation {z=3.273}
   }
   n49 = Get handAlphaDown_l_inv()
   n50 = Look At - Bone: dyn_sleeve_tp_01_l_MSTR(ComponentPose=@n48, Alpha=@n49) {
    alphaBoolBlend {blendOption=Linear}
    boneToModify {boneName=dyn_sleeve_tp_01_l_MSTR}
    lookAtLocation {x=0}
    lookAtTarget {boneReference {boneName=dyn_sleeve_tp_02_l_aim}}
    lookAt_Axis {axis=<1 0>}
    lookUp_Axis {axis {x=0, z=1}}
   }
   n51 = "Transform (Modify) Bone - Bone: dyn_sleeve_fr_02_l_aim"(ComponentPose=@n50, Alpha=1, Translation="-0.000,-0.000,3.734", Rotation="0,0,0", Scale="1,1,1") {
    translationMode=BMM_Additive
    translationSpace=BCS_BoneSpace
    alphaBoolBlend {blendOption=Linear}
    boneToModify {boneName=dyn_sleeve_fr_02_l_aim}
    translation {z=3.734}
   }
   n52 = Look At - Bone: dyn_sleeve_fr_01_l_MSTR(ComponentPose=@n51, Alpha=@n49) {
    alphaBoolBlend {blendOption=Linear}
    boneToModify {boneName=dyn_sleeve_fr_01_l_MSTR}
    lookAtLocation {x=0}
    lookAtTarget {boneReference {boneName=dyn_sleeve_fr_02_l_aim}}
    lookAt_Axis {axis=<1 0>}
    lookUp_Axis {axis {x=0, z=1}}
   }
   n53 = "Transform (Modify) Bone - Bone: dyn_sleeve_bk_02_l_aim"(ComponentPose=@n52, Alpha=1, Translation="0.000,0.000,3.895", Rotation="0,0,0", Scale="1,1,1") {
    translationMode=BMM_Additive
    translationSpace=BCS_BoneSpace
    alphaBoolBlend {blendOption=Linear}
    boneToModify {boneName=dyn_sleeve_bk_02_l_aim}
    translation {z=3.895}
   }
   n54 = Look At - Bone: dyn_sleeve_bk_01_l_MSTR(ComponentPose=@n53, Alpha=@n49) {
    alphaBoolBlend {blendOption=Linear}
    boneToModify {boneName=dyn_sleeve_bk_01_l_MSTR}
    lookAtLocation {x=0}
    lookAtTarget {boneReference {boneName=dyn_sleeve_bk_02_l_aim}}
    lookAt_Axis {axis=<1 0>}
    lookUp_Axis {axis {x=0, z=1}}
   }
   n55 = "Transform (Modify) Bone - Bone: dyn_sleeve_bt_02_l_aim"(ComponentPose=@n54, Alpha=1, Translation="0.000,-0.000,3.292", Rotation="0,0,0", Scale="1,1,1") {
    translationMode=BMM_Additive
    translationSpace=BCS_BoneSpace
    alphaBoolBlend {blendOption=Linear}
    boneToModify {boneName=dyn_sleeve_bt_02_l_aim}
    translation=<0 0 3.2923>
   }
   n56 = Look At - Bone: dyn_sleeve_bt_01_l_MSTR(ComponentPose=@n55, Alpha=@n49) {
    alphaBoolBlend {blendOption=Linear}
    boneToModify {boneName=dyn_sleeve_bt_01_l_MSTR}
    lookAtLocation {x=0}
    lookAtTarget {boneReference {boneName=dyn_sleeve_bt_02_l_aim}}
    lookAt_Axis {axis=<1 0>}
    lookUp_Axis {axis {x=0, z=1}}
   }
   n57 = "Constraint - dyn_hoodienew_01_l (P)"(ComponentPose=@n56, Alpha=0.5, ConstraintWeights_0=1) {
    alphaBoolBlend {blendOption=Linear}
    boneToModify {boneName=dyn_hoodienew_01_l}
    constraintSetup [{targetBone {boneName=pelvis}}]
    constraintWeights [1]
   }
   n58 = "Constraint - dyn_hoodienew_01_r (P)"(ComponentPose=@n57, Alpha=0.5, ConstraintWeights_0=1) {
    alphaBoolBlend {blendOption=Linear}
    boneToModify {boneName=dyn_hoodienew_01_r}
    constraintSetup [{targetBone {boneName=pelvis}}]
    constraintWeights [1]
   }
   n59 = Get isCrouchingAlpha()
   n60 = "Transform (Modify) Bone - Bone: dyn_hoodienew_01_l"(ComponentPose=@n58, Alpha=@n59, Translation="-3.512,3.512,-0.001", Rotation="-0.002,-35,-0.004", Scale="1,1,1") {
    rotationMode=BMM_Additive
    rotationSpace=BCS_BoneSpace
    translationMode=BMM_Additive
    translationSpace=BCS_BoneSpace
    alphaBoolBlend {blendOption=Linear}
    alphaScaleBiasClamp {scale=0.5}
    boneToModify {boneName=dyn_hoodienew_01_l}
    rotation=<rot -0.002 -35 -0.004>
    translation=<-3.512 3.512 -0.001>
   }
   n61 = "Transform (Modify) Bone - Bone: dyn_hoodienew_01_r"(ComponentPose=@n60, Alpha=@n59, Translation="-2.470,-2.470,0.000", Rotation="0.002,35,-0.004", Scale="1,1,1") {
    rotationMode=BMM_Additive
    rotationSpace=BCS_BoneSpace
    translationMode=BMM_Additive
    translationSpace=BCS_BoneSpace
    alphaBoolBlend {blendOption=Linear}
    alphaScaleBiasClamp {scale=0.5}
    boneToModify {boneName=dyn_hoodienew_01_r}
    rotation=<rot 0.002 35 -0.004>
    translation=<-2.47 -2.47>
   }
   n62 = Get windVector()
   n63 = Anim Dynamics - Bone: dyn_hoodienew_01_r!(ComponentPose=@n61, Alpha=0.5, GravityOverride:Vector=@n62, ExternalForce="0,0,0") -> (Pose) {
    angularDampingOverride=1
    angularSpringConstant=1500
    bAngularSpring=true
    bGravityOverrideInSimSpace=true
    bOverrideAngularDamping=true
    bUseGravityOverride=true
    gravityScale=0.35
    simulationSpace=RootRelative
    alphaBoolBlend {blendOption=Linear}
    boundBone {boneName=dyn_hoodienew_01_r}
    componentLinearAccScale=<0.3 0.3 0.3>
    componentLinearVelScale=<0.3 0.3 0.3>
    gravityOverride {z=-980}
    physicsBodyDefinitions [{
      sphereCollisionRadius=0
      boundBone {boneName=dyn_hoodienew_01_r}
      boxExtents=<10 20 20>
      constraintSetup {angularLimitsMax=<0 0 45>, angularLimitsMin=<0 0 0>, angularTarget=<0 0 0>}
      localJointOffset=<10 0 0>
     }]
   }
   n64 = Anim Dynamics - Bone: dyn_hoodienew_01_l!(ComponentPose=@n63, Alpha=0.5, GravityOverride:Vector=@n62, ExternalForce="0,0,0") -> (Pose) {
    angularDampingOverride=1
    angularSpringConstant=1500
    bAngularSpring=true
    bGravityOverrideInSimSpace=true
    bOverrideAngularDamping=true
    bUseGravityOverride=true
    gravityScale=0.35
    simulationSpace=RootRelative
    alphaBoolBlend {blendOption=Linear}
    boundBone {boneName=dyn_hoodienew_01_l}
    componentLinearAccScale=<0.3 0.3 0.3>
    componentLinearVelScale=<0.3 0.3 0.3>
    gravityOverride {z=-980}
    physicsBodyDefinitions [{
      sphereCollisionRadius=0
      boundBone {boneName=dyn_hoodienew_01_l}
      boxExtents=<10 20 20>
      constraintSetup {angularLimitsMax=<0 0 0>, angularLimitsMin=<0 0 -45>, angularTarget=<0 0 0>}
      localJointOffset=<10 0 0>
     }]
   }
   n65 = "Constraint - dyn_pelvisnew_01 (RR)"!(ComponentPose=@n64, Alpha=0.4, ConstraintWeights_0=0.5, ConstraintWeights_1:float=0.5) -> (Pose) {
    alphaBoolBlend {blendOption=Linear}
    boneToModify {boneName=dyn_pelvisnew_01}
    constraintSetup [{transformType=Rotation, targetBone {boneName=dyn_hoodienew_01_l}}, {transformType=Rotation, targetBone {boneName=dyn_hoodienew_01_r}}]
    constraintWeights [1,1]
   }
   n66 = "Constraint - dyn_boot_01_l (P)"(ComponentPose=@n65, Alpha=1, ConstraintWeights_0=1) {
    alphaBoolBlend {blendOption=Linear}
    boneToModify {boneName=dyn_boot_01_l}
    constraintSetup [{targetBone {boneName=foot_l}}]
    constraintWeights [1]
   }
   n67 = "Constraint - dyn_boot_01_l (R)"(ComponentPose=@n66, Alpha=0.5, ConstraintWeights_0=1) {
    alphaBoolBlend {blendOption=Linear}
    boneToModify {boneName=dyn_boot_01_l}
    constraintSetup [{transformType=Rotation, targetBone {boneName=calf_twist_02_l}}]
    constraintWeights [1]
   }
   n68 = "Constraint - dyn_boot_01_r (P)"(ComponentPose=@n67, Alpha=1, ConstraintWeights_0=1) {
    alphaBoolBlend {blendOption=Linear}
    boneToModify {boneName=dyn_boot_01_r}
    constraintSetup [{targetBone {boneName=foot_r}}]
    constraintWeights [1]
   }
   n69 = "Constraint - dyn_boot_01_r (R)"(ComponentPose=@n68, Alpha=0.5, ConstraintWeights_0=1) {
    alphaBoolBlend {blendOption=Linear}
    boneToModify {boneName=dyn_boot_01_r}
    constraintSetup [{transformType=Rotation, targetBone {boneName=calf_twist_01_r}}]
    constraintWeights [1]
   }
   n70 = "Transform (Modify) Bone - Bone: dyn_pelvisnew_01"(ComponentPose=@n69, Alpha=1, Translation="0.000,1.434,0.000", Rotation="0,0,0", Scale="1,1,1") {
    translationMode=BMM_Additive
    translationSpace=BCS_BoneSpace
    alphaBoolBlend {blendOption=Linear}
    boneToModify {boneName=dyn_pelvisnew_01}
    translation {y=1.434}
   }
   Output Pose(Result=Component To Local(ComponentPose=@n70)) {layerGroup=None}
  ]}
 EventGraph {nodes [
   Event BlueprintUpdateAnimation()
   > n1 = Sequence!(execute) -> (then_0, then_1, then_2:Exec, then_3:Exec, then_4:Exec)
   > Set isCrouchingAlpha(isCrouchingAlpha:double="To Float (Boolean)"(InBool=Get bIsCrouching())) -> (Output_Get:double)
   n2 = vector * float(
    A=Break Fort Anim Input Skydiving External Force(FortAnimInput_SkydivingExternalForce=Get SkydivingExternalForceInput()) -> (
     bUseSkydivingVectorForce=~
     SkydivingVectorForce=~
     HeadToPelvisDirection=~
     FloatingMultiplier=~
     FloatingAdditive=~
     DivingMultiplier=~
     DivingAdditive=~
     ParachutingMultiplier=~
     ParachutingAdditive=~
    ).SkydivingVectorForce
    B=0.15
   )
   n3 = Get bIsSkyDiving()
   n4 = GetOwnerSkeletalMeshComponent()
   n5 = BreakRotator(InRot=BreakTransform(InTransform=GetDeltaTransformFromRefPose(self=@n4, BoneName=pelvis, BaseName=root)).Rotation)
   n6 = SelectFloat(A=980, B=-980, bPickA="float > float"(A="Absolute (Float)"(A=@n5.Yaw), B=90) {is=title K2Node_CallFunction})
   Set windVector(
    execute=@n1.then_1
    windVector:Vector=SelectVector(A=MakeVector(Y=0, Z=@n6), B=SelectVector(A=@n2, B="0, 0,-980", bPickA=@n3), bPickA=Get bIsPlayingEmote())
   ) -> (Output_Get:Vector)
   n7 = GetOwnerSkeletalMeshComponent()
   Set handAlphaDown_r(
    execute=@n1.then_2
    handAlphaDown_r:double=GetAlphaFromReferenceTransformDelta(SkeletalMeshComponent=@n7, BoneName=hand_r, BaseName=lowerarm_r, InputRange=(X=-15,Y=10), OutputRange=(X=1,Y=0), TwistAxis="1,0,0")
   ) -> (Output_Get:double)
   > Set handAlphaDown_r_inv(handAlphaDown_r_inv:double=GetAlphaFromReferenceTransformDelta(SkeletalMeshComponent=@n7, BoneName=hand_r, BaseName=lowerarm_r, InputRange=(X=0,Y=40), TwistAxis="1,0,0")) -> (Output_Get:double)
   > Set handAlphaDown_l(handAlphaDown_l:double=GetAlphaFromReferenceTransformDelta(
     SkeletalMeshComponent=@n7
     BoneName=lowerarm_twist_02_l
     BaseName=lowerarm_l
     DeltaComponentType=EulerX
     InputRange=(X=-48,Y=-35)
     OutputRange=(X=1,Y=0)
     TwistAxis="1,0,0"
    )) -> (Output_Get:double)
   > Set handAlphaDown_l_inv(handAlphaDown_l_inv:double=GetAlphaFromReferenceTransformDelta(
     SkeletalMeshComponent=@n7
     BoneName=lowerarm_twist_02_l
     BaseName=lowerarm_l
     DeltaComponentType=EulerX
     InputRange=(X=-48,Y=-35)
     TwistAxis="1,0,0"
    )) -> (Output_Get:double)
   > Set thigh_l_alpha(thigh_l_alpha:double=GetAlphaFromReferenceTransformDelta(
     SkeletalMeshComponent=@n7
     BoneName=thigh_l
     BaseName=pelvis
     DeltaComponentType=EulerX
     InputRange=(X=0,Y=15)
     OutputRange=(X=1,Y=0)
     TwistAxis="1,0,0"
    )) -> (Output_Get:double)
   n8 = "integer > integer"(A=Get CurrentLOD(), B=1) {is=title K2Node_CallFunction}
   n9 = NOT Boolean(A=Get bRigidBodyEnabled())
   Set isMobileLODAlpha(execute=@n1.then_3, isMobileLODAlpha:double="To Float (Boolean)"(InBool=OR Boolean(A=@n9, B=@n8))) -> (Output_Get:double)
   n10 = Branch(execute=@n1.then_4, Condition=Get bIsTargeting())
   > Set Aim_Off(Aim_Off:double=0) -> (Output_Get:double)
   Set Aim_Off(execute=@n10.else, Aim_Off=1)
   Event OnEndSkydiving()
   > ResetDynamics(InTeleportType=TeleportPhysics)
   > ResetDynamics(InTeleportType=ResetPhysics)
  ]}
]
variables [
 Aim_Off {type=number}
 handAlphaDown_l {type=number}
 handAlphaDown_l_inv {type=number}
 handAlphaDown_r {type=number}
 handAlphaDown_r_inv {type=number}
 isCrouchingAlpha {type=number}
 isMobileLODAlpha {type=number}
 thigh_l_alpha {type=number}
 windVector {type=Vector}
]
