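# Physics asset for the hair strands and hood on the face accessory part of the female medium Hoodie Bandit cosmetic body.
# Capsules on head, neck and upper spine form the body, with box and capsule shapes on the dynamic bones.
# Tiny MSTR anchor bodies join head or neck with free rotation, each holding a narrowly limited dynamic bone.
# The top-right strand constrains directly to head, and front hood sides link to the center.
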PhysicsAsset /BRCosmetics/Characters/Player/Female/Medium/Bodies/F_MED_Hoodie_Bandit/Meshes/Parts/F_MED_Hoodie_Bandit_FaceAcc_Physics_hair
bodies {
 dyn_hairstrand_bt_01_l [{shapeType=Box, extentX=14.5, extentY=3, extentZ=4.5, center=<6.8875 0 0>}]
 dyn_hairstrand_bt_01_l_MSTR [{shapeType=Capsule, length=0.01, radius=0.01}]
 dyn_hairstrand_tp_01 [{shapeType=Capsule, length=10, radius=1.0232, shapeName=dyn_hairstrand_tp_01_l_capsule, center=<5.5 0 0>, rotation=<rot 90 0 0>}]
 dyn_hairstrand_tp_01_MSTR [{shapeType=Capsule, length=0.01, radius=0.01}]
 dyn_hairstrand_tp_01_l [{shapeType=Capsule, length=7.0356, radius=1.2732, shapeName=dyn_hairstrand_tp_01_l_capsule, center=<4 0 0>, rotation=<rot 90 0 0>}]
 dyn_hairstrand_tp_01_l_MSTR [{shapeType=Capsule, length=0.01, radius=0.01}]
 dyn_hairstrand_tp_01_r [{shapeType=Box, extentX=14.5, extentY=2, extentZ=4, center=<6.8875 0 0>}]
 dyn_hoodie_01_l [{shapeType=Box, extentX=10.5, extentY=5, extentZ=0.5, center=<4.8854 0 -0.0001>}]
 dyn_hoodie_01_l_MSTR [{shapeType=Capsule, length=0.01, radius=0.01}]
 dyn_hoodie_01_r [{shapeType=Box, extentX=10.5, extentY=5, extentZ=0.5, center=<4.8854 0 -0.0001>}]
 dyn_hoodie_01_r_MSTR [{shapeType=Capsule, length=0.01, radius=0.01}]
 dyn_hoodie_bk_01_l [{shapeType=Box, extentX=10.5, extentY=5, extentZ=0.5, center=<4.8854 0 -0.0001>}]
 dyn_hoodie_bk_01_l_MSTR [{shapeType=Capsule, length=0.01, radius=0.01}]
 dyn_hoodie_bk_01_r [{shapeType=Box, extentX=10.5, extentY=5, extentZ=0.5, center=<4.8854 0 -0.0001>}]
 dyn_hoodie_bk_01_r_MSTR [{shapeType=Capsule, length=0.01, radius=0.01}]
 dyn_hoodie_bt_01 [{shapeType=Box, extentX=10.5, extentY=5, extentZ=0.5, center=<4.8854 0 -0.0001>}]
 dyn_hoodie_bt_01_MSTR [{shapeType=Capsule, length=0.01, radius=0.01}]
 dyn_hoodie_bt_01_l [{shapeType=Box, extentX=10.5, extentY=5, extentZ=0.5, center=<4.8854 0 -0.0001>}]
 dyn_hoodie_bt_01_l_MSTR [{shapeType=Capsule, length=0.01, radius=0.01}]
 dyn_hoodie_bt_01_r [{shapeType=Box, extentX=10.5, extentY=5, extentZ=0.5, center=<4.8854 0 -0.0001>}]
 dyn_hoodie_bt_01_r_MSTR [{shapeType=Capsule, length=0.01, radius=0.01}]
 dyn_hoodie_fr_01 [{shapeType=Box, extentX=10.5, extentY=5, extentZ=0.5, center=<4.8854 0 -0.0001>}]
 dyn_hoodie_fr_01_MSTR [{shapeType=Capsule, length=0.01, radius=0.01}]
 dyn_hoodie_fr_01_l [{shapeType=Box, extentX=10.5, extentY=5, extentZ=0.5, center=<4.8854 0 -0.0001>}]
 dyn_hoodie_fr_01_l_MSTR [{shapeType=Capsule, length=0.01, radius=0.01}]
 dyn_hoodie_fr_01_r [{shapeType=Box, extentX=10.5, extentY=5, extentZ=0.5, center=<4.8854 0 -0.0001>}]
 dyn_hoodie_fr_01_r_MSTR [{shapeType=Capsule, length=0.01, radius=0.01}]
 dyn_hoodie_tp_01 [{shapeType=Box, extentX=10.5, extentY=5, extentZ=0.5, center=<4.8854 0 -0.0001>}]
 dyn_hoodie_tp_01_MSTR [{shapeType=Capsule, length=0.01, radius=0.01}]
 head [
  {shapeType=Capsule, length=11.75, radius=4.829, center=<-3.2273 -1.8718 -0.0001>, rotation=<rot 90 0 -10.0002>}
  {shapeType=Capsule, length=9.5, radius=4.579, center=<-0.6443 3.0169 -0.0001>, rotation=<rot 90 0 -10.0002>}
 ]
 neck_01 [{shapeType=Capsule, length=2.8083, radius=4.25, center=<3.8842 0.6051 0>, rotation=<rot 90 0.0002 0>}]
 spine_05 [{shapeType=Capsule, length=10.9792, radius=10.079, center=<-2.3057 -0.8568 0>}]
}
constraints [
 {bone1Name=neck_01, bone2Name=spine_05, swing1LimitDegrees=45, swing2LimitDegrees=45, twistLimitDegrees=45}
 {bone1Name=head, bone2Name=neck_01, swing1LimitDegrees=45, swing2LimitDegrees=45, twistLimitDegrees=45}
 {
  bone1Name=dyn_hairstrand_bt_01_l_MSTR
  bone2Name=head
  swing1LimitDegrees=45
  swing1Motion=Free
  swing2LimitDegrees=45
  swing2Motion=Free
  twistLimitDegrees=45
  twistMotion=Free
 }
 {
  bone1Name=dyn_hoodie_01_l_MSTR
  bone2Name=head
  swing1LimitDegrees=45
  swing1Motion=Free
  swing2LimitDegrees=45
  swing2Motion=Free
  twistLimitDegrees=45
  twistMotion=Free
 }
 {
  bone1Name=dyn_hairstrand_tp_01_l_MSTR
  bone2Name=head
  swing1LimitDegrees=45
  swing1Motion=Free
  swing2LimitDegrees=45
  swing2Motion=Free
  twistLimitDegrees=45
  twistMotion=Free
 }
 {
  bone1Name=dyn_hoodie_fr_01_l_MSTR
  bone2Name=head
  swing1LimitDegrees=45
  swing1Motion=Free
  swing2LimitDegrees=45
  swing2Motion=Free
  twistLimitDegrees=45
  twistMotion=Free
 }
 {
  bone1Name=dyn_hairstrand_tp_01_MSTR
  bone2Name=head
  swing1LimitDegrees=45
  swing1Motion=Free
  swing2LimitDegrees=45
  swing2Motion=Free
  twistLimitDegrees=45
  twistMotion=Free
 }
 {
  bone1Name=dyn_hoodie_fr_01_r_MSTR
  bone2Name=head
  swing1LimitDegrees=45
  swing1Motion=Free
  swing2LimitDegrees=45
  swing2Motion=Free
  twistLimitDegrees=45
  twistMotion=Free
 }
 {
  bone1Name=dyn_hoodie_01_r_MSTR
  bone2Name=head
  swing1LimitDegrees=45
  swing1Motion=Free
  swing2LimitDegrees=45
  swing2Motion=Free
  twistLimitDegrees=45
  twistMotion=Free
 }
 {
  bone1Name=dyn_hoodie_fr_01_MSTR
  bone2Name=head
  swing1LimitDegrees=45
  swing1Motion=Free
  swing2LimitDegrees=45
  swing2Motion=Free
  twistLimitDegrees=45
  twistMotion=Free
 }
 {
  bone1Name=dyn_hoodie_tp_01_MSTR
  bone2Name=head
  swing1LimitDegrees=45
  swing1Motion=Free
  swing2LimitDegrees=45
  swing2Motion=Free
  twistLimitDegrees=45
  twistMotion=Free
 }
 {
  bone1Name=dyn_hoodie_bt_01_l_MSTR
  bone2Name=neck_01
  swing1LimitDegrees=45
  swing1Motion=Free
  swing2LimitDegrees=45
  swing2Motion=Free
  twistLimitDegrees=45
  twistMotion=Free
 }
 {
  bone1Name=dyn_hoodie_bt_01_MSTR
  bone2Name=neck_01
  swing1LimitDegrees=45
  swing1Motion=Free
  swing2LimitDegrees=45
  swing2Motion=Free
  twistLimitDegrees=45
  twistMotion=Free
 }
 {
  bone1Name=dyn_hoodie_bt_01_r_MSTR
  bone2Name=neck_01
  swing1LimitDegrees=45
  swing1Motion=Free
  swing2LimitDegrees=45
  swing2Motion=Free
  twistLimitDegrees=45
  twistMotion=Free
 }
 {
  bone1Name=dyn_hoodie_bk_01_l_MSTR
  bone2Name=neck_01
  swing1LimitDegrees=45
  swing1Motion=Free
  swing2LimitDegrees=45
  swing2Motion=Free
  twistLimitDegrees=45
  twistMotion=Free
 }
 {
  bone1Name=dyn_hoodie_bk_01_r_MSTR
  bone2Name=neck_01
  swing1LimitDegrees=45
  swing1Motion=Free
  swing2LimitDegrees=45
  swing2Motion=Free
  twistLimitDegrees=45
  twistMotion=Free
 }
 {bone1Name=dyn_hoodie_bk_01_r, bone2Name=dyn_hoodie_bk_01_r_MSTR, swing1LimitDegrees=5, swing2LimitDegrees=15, twistLimitDegrees=5}
 {bone1Name=dyn_hairstrand_tp_01_r, bone2Name=head, swing1LimitDegrees=5, swing2LimitDegrees=5, twistLimitDegrees=5}
 {bone1Name=dyn_hairstrand_bt_01_l, bone2Name=dyn_hairstrand_bt_01_l_MSTR, swing1LimitDegrees=5, swing2LimitDegrees=5, twistLimitDegrees=5}
 {bone1Name=dyn_hoodie_01_l, bone2Name=dyn_hoodie_01_l_MSTR, swing1LimitDegrees=5, swing2LimitDegrees=15, twistLimitDegrees=5}
 {bone1Name=dyn_hairstrand_tp_01_l, bone2Name=dyn_hairstrand_tp_01_l_MSTR, swing1LimitDegrees=15, swing2LimitDegrees=15, twistLimitDegrees=5}
 {bone1Name=dyn_hoodie_fr_01_l, bone2Name=dyn_hoodie_fr_01_l_MSTR, swing1LimitDegrees=5, swing2LimitDegrees=15, twistLimitDegrees=5}
 {bone1Name=dyn_hairstrand_tp_01, bone2Name=dyn_hairstrand_tp_01_MSTR, swing1LimitDegrees=20, swing2LimitDegrees=15, twistLimitDegrees=5}
 {bone1Name=dyn_hoodie_fr_01_r, bone2Name=dyn_hoodie_fr_01_r_MSTR, swing1LimitDegrees=5, swing2LimitDegrees=15, twistLimitDegrees=5}
 {bone1Name=dyn_hoodie_01_r, bone2Name=dyn_hoodie_01_r_MSTR, swing1LimitDegrees=5, swing2LimitDegrees=15, twistLimitDegrees=5}
 {bone1Name=dyn_hoodie_fr_01, bone2Name=dyn_hoodie_fr_01_MSTR, swing1LimitDegrees=25, swing2LimitDegrees=5, twistLimitDegrees=5}
 {bone1Name=dyn_hoodie_tp_01, bone2Name=dyn_hoodie_tp_01_MSTR, swing1LimitDegrees=5, swing2LimitDegrees=5, twistLimitDegrees=5}
 {bone1Name=dyn_hoodie_bt_01_l, bone2Name=dyn_hoodie_bt_01_l_MSTR, swing1LimitDegrees=5, swing2LimitDegrees=25, twistLimitDegrees=5}
 {bone1Name=dyn_hoodie_bt_01, bone2Name=dyn_hoodie_bt_01_MSTR, swing1LimitDegrees=5, swing2LimitDegrees=25, twistLimitDegrees=5}
 {bone1Name=dyn_hoodie_bt_01_r, bone2Name=dyn_hoodie_bt_01_r_MSTR, swing1LimitDegrees=5, swing2LimitDegrees=25, twistLimitDegrees=5}
 {bone1Name=dyn_hoodie_bk_01_l, bone2Name=dyn_hoodie_bk_01_l_MSTR, swing1LimitDegrees=5, swing2LimitDegrees=15, twistLimitDegrees=5}
 {
  bone1Name=dyn_hoodie_fr_01_r
  bone2Name=dyn_hoodie_fr_01
  swing1LimitDegrees=45
  swing1Motion=Free
  swing2LimitDegrees=45
  swing2Motion=Free
  twistLimitDegrees=45
  twistMotion=Free
 }
 {
  bone1Name=dyn_hoodie_fr_01_l
  bone2Name=dyn_hoodie_fr_01
  swing1LimitDegrees=45
  swing1Motion=Free
  swing2LimitDegrees=45
  swing2Motion=Free
  twistLimitDegrees=45
  twistMotion=Free
 }
]
