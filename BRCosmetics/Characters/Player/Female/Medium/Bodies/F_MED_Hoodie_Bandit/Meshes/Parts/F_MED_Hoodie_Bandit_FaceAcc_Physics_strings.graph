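# Physics asset for the face accessory part of the female medium Hoodie Bandit body, simulating hoodie drawstrings.
# Capsule bodies cover two chained cord bones, plus the neck and upper spine as body volumes.
# The first cord segment is constrained to spine_05 and the second to the first, with moderate swing and tight twist limits.
# The neck joins spine_05 through a looser constraint.
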PhysicsAsset /BRCosmetics/Characters/Player/Female/Medium/Bodies/F_MED_Hoodie_Bandit/Meshes/Parts/F_MED_Hoodie_Bandit_FaceAcc_Physics_strings
bodies {
 dyn_cord_01 [{shapeType=Capsule, length=3, radius=2, shapeName=dyn_cord_01_capsule, center=<3 0 1.6039>, rotation=<rot 90 0 0>}]
 dyn_cord_02 [{shapeType=Capsule, length=3, radius=2, shapeName=dyn_cord_01_capsule, center=<3 0 1.6039>, rotation=<rot 90 0 0>}]
 neck_01 [{shapeType=Capsule, length=2.8083, radius=4.25, center=<3.8842 0.6051 0>, rotation=<rot 90 0.0002 0>}]
 spine_05 [
  {shapeType=Capsule, length=10.9792, radius=10.3, center=<-2.3057 -0.8568 0>}
  {shapeType=Capsule, length=10, radius=10, shapeName=spine_05_capsule, center=<-6.0793 1.1838 0>}
 ]
}
constraints [
 {bone1Name=neck_01, bone2Name=spine_05, swing1LimitDegrees=45, swing2LimitDegrees=45, twistLimitDegrees=45}
 {bone1Name=dyn_cord_01, bone2Name=spine_05, swing1LimitDegrees=25, swing2LimitDegrees=25, twistLimitDegrees=5}
 {bone1Name=dyn_cord_02, bone2Name=dyn_cord_01, swing1LimitDegrees=25, swing2LimitDegrees=25, twistLimitDegrees=5}
]
