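# Animation Blueprint for the Bandit hoodie's face-accessory part, adding secondary motion to hood, strings and hair.
# It copies the attached parent's pose, drives look-at bones by head-versus-spine rotation, then simulates strings and hair with RigidBody nodes using skydiving-aware gravity.
# A left hair strand is constrained to the head, and weapon type plus aiming or relaxed state rotates right-side hood and strand bones.
# Dynamics reset when skydiving ends.
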AnimBlueprint /BRCosmetics/Characters/Player/Female/Medium/Bodies/F_MED_Hoodie_Bandit/Meshes/Parts/F_MED_Hoodie_Bandit_FaceAcc_AnimBlueprint
target_skeleton=&/F_MED_Hoodie_Bandit_FaceAcc_Skeleton
graphs [
 AnimGraph {nodes [
   "Constraint - dyn_hoodie_tp_02_aim (P)"(Alpha=1, ConstraintWeights_0=1) {
    alphaBoolBlend {blendOption=Linear}
    boneToModify {boneName=dyn_hoodie_tp_02_aim}
    constraintSetup [{targetBone {boneName=neck_02}}]
    constraintWeights [1]
   }
   n1 = Get velocityAlpha() -> (velocityAlpha:double)
   n2 = "Constraint - dyn_hairstrand_tp_01_r (P)"(Alpha=@n1, ConstraintWeights_0=1) {
    alphaBoolBlend {blendOption=Linear}
    boneToModify {boneName=dyn_hairstrand_tp_01_r}
    constraintSetup [{targetBone {boneName=head}}]
    constraintWeights [1]
   }
   n3 = "Constraint - dyn_hairstrand_tp_01 (P)"(ComponentPose=@n2, Alpha=@n1, ConstraintWeights_0=1) {
    alphaBoolBlend {blendOption=Linear}
    boneToModify {boneName=dyn_hairstrand_tp_01}
    constraintSetup [{targetBone {boneName=head}}]
    constraintWeights [1]
   }
   "Constraint - dyn_hairstrand_tp_01_l (P)"(ComponentPose=@n3, Alpha=@n1, ConstraintWeights_0=1) {
    alphaBoolBlend {blendOption=Linear}
    boneToModify {boneName=dyn_hairstrand_tp_01_l}
    constraintSetup [{targetBone {boneName=head}}]
    constraintWeights [1]
   }
   n4 = Get windVector() -> (windVector:Vector)
   RigidBody!(ComponentPose, Alpha, OverrideWorldGravity:Vector=@n4, ExternalForce) -> (Pose) {
    bClampLinearTranslationLimitToRefPose=true
    bOverrideWorldGravity=true
    overridePhysicsAsset=&/F_MED_Hoodie_Bandit_FaceAcc_Physics
    simulationSpace=BaseBoneSpace
    alphaBoolBlend {blendOption=Linear}
    baseBoneRef {boneName=root}
    componentAppliedLinearAccClamp=<3000 3000 3000>
    componentLinearAccScale=<0.5 0.5 0.5>
    componentLinearVelScale=<0.7 0.7 0.7>
    overrideWorldGravity {z=-980}
   }
   n5 = "Transform (Modify) Bone - Bone: dyn_hair_bk_02_l_aim"(
    ComponentPose=Local To Component(LocalPose=Copy Pose From Mesh() {bUseAttachedParent=true})
    Alpha=1
    Translation="0,-2.5,0"
    Rotation="0,0,0"
    Scale="1,1,1"
   ) {
    translationMode=BMM_Additive
    translationSpace=BCS_BoneSpace
    alphaBoolBlend {blendOption=Linear}
    boneToModify {boneName=dyn_hair_bk_02_l_aim}
    translation {y=-2.5}
   }
   n6 = "Transform (Modify) Bone - Bone: dyn_hair_bk_02_r_aim"(ComponentPose=@n5, Alpha=1, Translation="0,2.5,0", Rotation="0,0,0", Scale="1,1,1") {
    translationMode=BMM_Additive
    translationSpace=BCS_BoneSpace
    alphaBoolBlend {blendOption=Linear}
    boneToModify {boneName=dyn_hair_bk_02_r_aim}
    translation {y=2.5}
   }
   n7 = "Transform (Modify) Bone - Bone: dyn_hair_bk_02_aim"(ComponentPose=@n6, Alpha=1, Translation="5,-0.000,0", Rotation="0,0,0", Scale="1,1,1") {
    translationMode=BMM_Additive
    translationSpace=BCS_BoneSpace
    alphaBoolBlend {blendOption=Linear}
    boneToModify {boneName=dyn_hair_bk_02_aim}
    translation {x=5}
   }
   n8 = Get headAlpha() -> (headAlpha:double)
   n9 = Look At - Bone: dyn_hair_bk_01(ComponentPose=@n7, Alpha=@n8) {
    alphaBoolBlend {blendOption=Linear}
    boneToModify {boneName=dyn_hair_bk_01}
    lookAtLocation {x=0}
    lookAtTarget {boneReference {boneName=dyn_hair_bk_02_aim}}
    lookAt_Axis {axis=<1 0>}
    lookUp_Axis {bInLocalSpace=false, axis {x=0, z=1}}
   }
   n10 = Look At - Bone: dyn_hair_bk_01_r(ComponentPose=@n9, Alpha=@n8) {
    alphaBoolBlend {blendOption=Linear}
    boneToModify {boneName=dyn_hair_bk_01_r}
    lookAtLocation {x=0}
    lookAtTarget {boneReference {boneName=dyn_hair_bk_02_r_aim}}
    lookAt_Axis {axis=<1 0>}
    lookUp_Axis {bInLocalSpace=false, axis {x=0, z=1}}
   }
   n11 = Look At - Bone: dyn_hair_bk_01_l(ComponentPose=@n10, Alpha=@n8) {
    alphaBoolBlend {blendOption=Linear}
    boneToModify {boneName=dyn_hair_bk_01_l}
    lookAtLocation {x=0}
    lookAtTarget {boneReference {boneName=dyn_hair_bk_02_l_aim}}
    lookAt_Axis {axis=<1 0>}
    lookUp_Axis {bInLocalSpace=false, axis {x=0, z=1}}
   }
   n12 = Get headAlphaB() -> (headAlphaB:double)
   n13 = "Transform (Modify) Bone - Bone: dyn_hoodie_tp_02_aim"(ComponentPose=@n11, Alpha=@n12, Translation="0.000,-0.000,3.5", Rotation="0,0,0", Scale="1,1,1") {
    translationMode=BMM_Additive
    translationSpace=BCS_BoneSpace
    alphaBoolBlend {blendOption=Linear}
    boneToModify {boneName=dyn_hoodie_tp_02_aim}
    translation {z=3.5}
   }
   n14 = Look At - Bone: dyn_hoodie_tp_02(ComponentPose=@n13, Alpha=@n12) {
    alphaBoolBlend {blendOption=Linear}
    boneToModify {boneName=dyn_hoodie_tp_02}
    lookAtLocation {x=0}
    lookAtTarget {boneReference {boneName=dyn_hoodie_tp_02_aim}}
    lookAt_Axis {axis=<1 0>}
    lookUp_Axis {bInLocalSpace=false, axis {x=0, z=1}}
   }
   n15 = RigidBody(ComponentPose=@n14, OverrideWorldGravity=@n4) {
    bClampLinearTranslationLimitToRefPose=true
    bOverrideWorldGravity=true
    overridePhysicsAsset=&/F_MED_Hoodie_Bandit_FaceAcc_Physics_strings
    simulationSpace=BaseBoneSpace
    alphaBoolBlend {blendOption=Linear}
    baseBoneRef {boneName=spine_05}
    componentAppliedLinearAccClamp=<2000 2000 2000>
    componentLinearAccScale=<0.3 0.3 0.4>
    componentLinearVelScale=<0.3 0.3 0.4>
    overrideWorldGravity {z=-980}
   }
   n16 = RigidBody(ComponentPose=@n15, OverrideWorldGravity=@n4) {
    bClampLinearTranslationLimitToRefPose=true
    bOverrideWorldGravity=true
    overridePhysicsAsset=&/F_MED_Hoodie_Bandit_FaceAcc_Physics_hair
    simulationSpace=BaseBoneSpace
    alphaBoolBlend {blendOption=Linear}
    baseBoneRef {boneName=head}
    componentAppliedLinearAccClamp=<2000 2000 2000>
    componentLinearAccScale=<0.3 0.3 0.35>
    componentLinearVelScale=<0.3 0.3 0.35>
    overrideWorldGravity {z=-980}
   }
   n17 = "Constraint - dyn_hairstrand_bt_01_l (P)"(ComponentPose=@n16, Alpha=1, ConstraintWeights_0=1) {
    alphaBoolBlend {blendOption=Linear}
    boneToModify {boneName=dyn_hairstrand_bt_01_l}
    constraintSetup [{targetBone {boneName=head}}]
    constraintWeights [1]
   }
   n18 = "Constraint - dyn_hairstrand_bt_01_l (P)"(ComponentPose=@n17, Alpha=@n1, ConstraintWeights_0=1) {
    alphaBoolBlend {blendOption=Linear}
    boneToModify {boneName=dyn_hairstrand_bt_01_l}
    constraintSetup [{targetBone {boneName=head}}]
    constraintWeights [1]
   }
   n19 = Get ar_firingPosition() -> (ar_firingPosition:double)
   n20 = "Transform (Modify) Bone - Bone: dyn_hoodie_bk_01_r"(ComponentPose=@n18, Alpha=@n19, Translation="0,0,0", Rotation="-55,0.000,0.000", Scale="1,1,1") {
    rotationMode=BMM_Additive
    rotationSpace=BCS_BoneSpace
    alphaBoolBlend {blendOption=Linear}
    boneToModify {boneName=dyn_hoodie_bk_01_r}
    rotation {pitch=-55}
   }
   n21 = "Transform (Modify) Bone - Bone: dyn_hoodie_01_r"(ComponentPose=@n20, Alpha=@n19, Translation="0,0,0", Rotation="-20,0.000,0.000", Scale="1,1,1") {
    rotationMode=BMM_Additive
    rotationSpace=BCS_BoneSpace
    alphaBoolBlend {blendOption=Linear}
    boneToModify {boneName=dyn_hoodie_01_r}
    rotation {pitch=-20}
   }
   n22 = "Transform (Modify) Bone - Bone: dyn_hairstrand_tp_01_r"(ComponentPose=@n21, Alpha=@n19, Translation="0,0,0", Rotation="-5,0.000,0.000", Scale="1,1,1") {
    rotationMode=BMM_Additive
    rotationSpace=BCS_BoneSpace
    alphaBoolBlend {blendOption=Linear}
    boneToModify {boneName=dyn_hairstrand_tp_01_r}
    rotation {pitch=-5}
   }
   Output Pose(Result=Component To Local(ComponentPose=@n22)) {layerGroup=None}
  ]}
 EventGraph {nodes [
   n1 = Cast To FortPlayerAnimInstance(execute=@n10.then_4, Object=GetAnimInstance(self=GetOwnerSkeletalMeshComponent()))
   n2 = Get WeaponCoreAnim(self=@n1)
   PrintString!(
    execute
    InString=Enum to String!(Enumerator:EFortWeaponCoreAnimation Enum=@n2) -> (ReturnValue:String)
    bPrintToScreen
    bPrintToLog
    TextColor
    Duration
    Key
   ) -> (then)
   > n3 = Set ar_firingPosition(ar_firingPosition:double=0) -> (Output_Get:double)
   n4 = Switch on EFortWeaponCoreAnimation!(execute=[@n1.then, @n3.then], Selection:EFortWeaponCoreAnimation Enum=@n2) -> (
    Melee:Exec
    Pistol:Exec
    Shotgun:Exec
    PaperBlueprint:Exec
    Rifle:Exec
    MeleeOneHand:Exec
    MachinePistol:Exec
    RocketLauncher:Exec
    GrenadeLauncher:Exec
    GoingCommando:Exec
    AssaultRifle:Exec
    TacticalShotgun:Exec
    SniperRifle:Exec
    TrapPlacement:Exec
    ShoulderLauncher:Exec
    AbilityDecoTool:Exec
    Crossbow:Exec
    C4:Exec
    RemoteControl:Exec
    DualWield:Exec
    AR_BullPup:Exec
    AR_ForwardGrip:Exec
    MedPackPaddles:Exec
    SMG_P90:Exec
    AR_DrumGun:Exec
    Consumable_Small:Exec
    Consumable_Large:Exec
    Balloon:Exec
    MountedTurret:Exec
    CreativeTool:Exec
    ExplosiveBow:Exec
    AshtonIndigo:Exec
    AshtonChicago:Exec
    MeleeDualWield:Exec
    Unarmed:Exec
   )
   n5 = Branch(execute=[@n4.AR_DrumGun, @n4.GrenadeLauncher, @n4.SMG_P90], Condition=Get bIsTargeting())
   Set ar_firingPosition(execute=[@n8.else, @n5.then, @n6.then, @n4.RocketLauncher], ar_firingPosition=1)
   Set ar_firingPosition(execute=[@n7.then, @n8.then, @n5.else], ar_firingPosition=0)
   n6 = Branch(execute=[@n4.AssaultRifle, @n4.Rifle, @n4.Shotgun, @n4.SniperRifle, @n4.TacticalShotgun], Condition=Get bIsTargeting())
   n7 = Branch(execute=@n6.else, Condition=Get bIsRelaxedLevel1() -> (bIsRelaxedLevel1:bool))
   n8 = Branch(execute=@n7.else, Condition=Get bIsRelaxedLevel2() -> (bIsRelaxedLevel2:bool))
   n9 = Event BlueprintUpdateAnimation()
   > n10 = Sequence!(execute) -> (then_0, then_1, then_2:Exec, then_3:Exec, then_4:Exec)
   > Set isTargetingAlpha(isTargetingAlpha:double="To Float (Boolean)"(InBool=Get bIsTargeting())) -> (Output_Get:double)
   n11 = Break Fort Anim Input Skydiving External Force(FortAnimInput_SkydivingExternalForce=Get SkydivingExternalForceInput()) -> (
    bUseSkydivingVectorForce=~
    SkydivingVectorForce=~
    HeadToPelvisDirection=~
    FloatingMultiplier=~
    FloatingAdditive=~
    DivingMultiplier=~
    DivingAdditive=~
    ParachutingMultiplier=~
    ParachutingAdditive=~
   )
   Set windVector(
    execute=@n10.then_1
    windVector:Vector=SelectVector(A=vector * float(A=@n11.SkydivingVectorForce, B=0.15), B="0, 0,-980", bPickA=Get bIsSkyDiving())
   ) -> (Output_Get:Vector)
   n12 = GetOwnerSkeletalMeshComponent()
   n13 = GetAlphaFromReferenceTransformDelta(SkeletalMeshComponent=@n12, BoneName=head, BaseName=spine_05, InputRange=(X=-10,Y=10), OutputRange=(X=1,Y=0), TwistAxis="1,0,0")
   Set headAlpha(execute=@n10.then_2, headAlpha:double=@n13) -> (Output_Get:double)
   > Set headAlphaB(headAlphaB:double=GetAlphaFromReferenceTransformDelta(SkeletalMeshComponent=@n12, BoneName=head, BaseName=spine_05, InputRange=(X=-35,Y=0), OutputRange=(X=1,Y=0), TwistAxis="1,0,0")) -> (Output_Get:double)
   n14 = GetPartSkeletalMeshComponent()
   n15 = String To Name(InString=SelectString(A=spine_05, B=neck_01, bPickA=Get bIsTargeting()))
   n16 = Calculate Velocity From Position History!(
    execute=@n10.then_3
    DeltaSeconds:float=@n9.DeltaTimeX
    Position:Vector=BreakTransform(InTransform=GetDeltaTransformFromRefPose(self=@n14, BoneName=head, BaseName=@n15))
    History:"Position History Structure (by ref)"=Make Position History!() -> (PositionHistory:Position History Structure) {is=title K2Node_MakeStruct}
    NumberOfSamples:int=3
    VelocityMin:float=0.f
    VelocityMax:float=5
   ) -> (then, ReturnValue:float)
   > Set velocityAlpha(velocityAlpha:double=SelectFloat(A=@n16, B=0, bPickA="float >= float"(A=@n16, B=0.2) {is=title K2Node_CallFunction})) -> (Output_Get:double)
   Event OnEndSkydiving()
   > ResetDynamics(InTeleportType=TeleportPhysics)
   > ResetDynamics(InTeleportType=ResetPhysics)
   TryGetPawnOwner()
   float * float(A=@n13, B=0.5)
  ]}
]
variables [
 ar_firingPosition {type=number}
 headAlpha {type=number}
 headAlphaB {type=number}
 isTargetingAlpha {type=number}
 velocityAlpha {type=number}
 windVector {type=Vector}
]
