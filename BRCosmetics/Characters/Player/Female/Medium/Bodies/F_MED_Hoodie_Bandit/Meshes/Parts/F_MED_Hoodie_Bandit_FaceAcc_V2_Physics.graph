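# Physics asset for the Bandit female hoodie FaceAcc V2 mesh, defining secondary-motion bodies for hair strands, hood pieces and a cord.
# Head and spine_05 capsules anchor tiny MSTR bodies through angularly free constraints; box bodies hang from most masters with tight swing and twist limits.
# The two-link cord chains straight from spine_05, and the top-right strand attaches directly to the head.
# The three back-hair masters have no child bodies here.
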PhysicsAsset /BRCosmetics/Characters/Player/Female/Medium/Bodies/F_MED_Hoodie_Bandit/Meshes/Parts/F_MED_Hoodie_Bandit_FaceAcc_V2_Physics
bodies {
 dyn_cord_01 [{shapeType=Box, extentX=3, extentY=7.5, extentZ=1.5, center=<1.4957 0 -0.0001>}]
 dyn_cord_02 [{shapeType=Box, extentX=3, extentY=7.5, extentZ=1.5, center=<1.4957 0 -0.0001>}]
 dyn_hair_bk_01_MSTR [{shapeType=Capsule, length=0.01, radius=0.01}]
 dyn_hair_bk_01_l_MSTR [{shapeType=Capsule, length=0.01, radius=0.01}]
 dyn_hair_bk_01_r_MSTR [{shapeType=Capsule, length=0.01, radius=0.01}]
 dyn_hairstrand_bt_01_l [{shapeType=Box, extentX=14.5, extentY=3, extentZ=4.5, center=<6.8875 0 0>}]
 dyn_hairstrand_bt_01_l_MSTR [{shapeType=Capsule, length=0.01, radius=0.01}]
 dyn_hairstrand_tp_01 [{shapeType=Box, extentX=14.5, extentY=3, extentZ=4.5, center=<6.8875 0 0>}]
 dyn_hairstrand_tp_01_MSTR [{shapeType=Capsule, length=0.01, radius=0.01}]
 dyn_hairstrand_tp_01_l [{shapeType=Box, extentX=14.5, extentY=4, extentZ=2, center=<6.8875 0 0>}]
 dyn_hairstrand_tp_01_l_MSTR [{shapeType=Capsule, length=0.01, radius=0.01}]
 dyn_hairstrand_tp_01_r [{shapeType=Box, extentX=14.5, extentY=3, extentZ=4.5, center=<6.8875 0 0>}]
 dyn_hoodie_bk_01_l [{shapeType=Box, extentX=10.5, extentY=5, extentZ=0.5, center=<4.8854 0 -0.0001>}]
 dyn_hoodie_bk_01_l_MSTR [{shapeType=Capsule, length=0.01, radius=0.01}]
 dyn_hoodie_bk_01_r [{shapeType=Box, extentX=10.5, extentY=5, extentZ=0.5, center=<4.8854 0 -0.0001>}]
 dyn_hoodie_bk_01_r_MSTR [{shapeType=Capsule, length=0.01, radius=0.01}]
 dyn_hoodie_bt_01 [{shapeType=Box, extentX=9, extentY=5, extentZ=1.5, center=<4.9556 -0.0001 -0.0001>}]
 dyn_hoodie_bt_01_MSTR [{shapeType=Capsule, length=0.01, radius=0.01}]
 dyn_hoodie_bt_01_l [{shapeType=Box, extentX=9, extentY=5, extentZ=1.5, center=<4.9556 -0.0001 -0.0001>}]
 dyn_hoodie_bt_01_l_MSTR [{shapeType=Capsule, length=0.01, radius=0.01}]
 dyn_hoodie_bt_01_r [{shapeType=Box, extentX=9, extentY=5, extentZ=1.5, center=<4.9556 -0.0001 -0.0001>}]
 dyn_hoodie_bt_01_r_MSTR [{shapeType=Capsule, length=0.01, radius=0.01}]
 head [
  {shapeType=Capsule, length=11.75, radius=4.829, center=<-3.2273 -1.8718 -0.0001>, rotation=<rot 90 0 -10.0002>}
  {shapeType=Capsule, length=8, radius=4.329, center=<-2.1653 4.813 -0.0001>, rotation=<rot 90 0 -10.0002>}
 ]
 spine_05 [{shapeType=Capsule, length=10.9792, radius=10.079, center=<-2.3057 -0.8568 0>}]
}
constraints [
 {bone1Name=head, bone2Name=spine_05, swing1LimitDegrees=45, swing2LimitDegrees=45, twistLimitDegrees=45}
 {
  bone1Name=dyn_hairstrand_bt_01_l_MSTR
  bone2Name=head
  swing1LimitDegrees=45
  swing1Motion=Free
  swing2LimitDegrees=45
  swing2Motion=Free
  twistLimitDegrees=45
  twistMotion=Free
 }
 {
  bone1Name=dyn_hairstrand_tp_01_l_MSTR
  bone2Name=head
  swing1LimitDegrees=45
  swing1Motion=Free
  swing2LimitDegrees=45
  swing2Motion=Free
  twistLimitDegrees=45
  twistMotion=Free
 }
 {
  bone1Name=dyn_hairstrand_tp_01_MSTR
  bone2Name=head
  swing1LimitDegrees=45
  swing1Motion=Free
  swing2LimitDegrees=45
  swing2Motion=Free
  twistLimitDegrees=45
  twistMotion=Free
 }
 {
  bone1Name=dyn_hoodie_bt_01_l_MSTR
  bone2Name=spine_05
  swing1LimitDegrees=45
  swing1Motion=Free
  swing2LimitDegrees=45
  swing2Motion=Free
  twistLimitDegrees=45
  twistMotion=Free
 }
 {
  bone1Name=dyn_hoodie_bt_01_MSTR
  bone2Name=spine_05
  swing1LimitDegrees=45
  swing1Motion=Free
  swing2LimitDegrees=45
  swing2Motion=Free
  twistLimitDegrees=45
  twistMotion=Free
 }
 {
  bone1Name=dyn_hoodie_bt_01_r_MSTR
  bone2Name=spine_05
  swing1LimitDegrees=45
  swing1Motion=Free
  swing2LimitDegrees=45
  swing2Motion=Free
  twistLimitDegrees=45
  twistMotion=Free
 }
 {
  bone1Name=dyn_hoodie_bk_01_l_MSTR
  bone2Name=spine_05
  swing1LimitDegrees=45
  swing1Motion=Free
  swing2LimitDegrees=45
  swing2Motion=Free
  twistLimitDegrees=45
  twistMotion=Free
 }
 {
  bone1Name=dyn_hoodie_bk_01_r_MSTR
  bone2Name=spine_05
  swing1LimitDegrees=45
  swing1Motion=Free
  swing2LimitDegrees=45
  swing2Motion=Free
  twistLimitDegrees=45
  twistMotion=Free
 }
 {bone1Name=dyn_cord_01, bone2Name=spine_05, swing1LimitDegrees=25, swing2LimitDegrees=25, twistLimitDegrees=5}
 {bone1Name=dyn_cord_02, bone2Name=dyn_cord_01, swing1LimitDegrees=25, swing2LimitDegrees=25, twistLimitDegrees=5}
 {bone1Name=dyn_hoodie_bk_01_r, bone2Name=dyn_hoodie_bk_01_r_MSTR, swing1LimitDegrees=5, swing2LimitDegrees=25, twistLimitDegrees=5}
 {bone1Name=dyn_hairstrand_tp_01_r, bone2Name=head, swing1LimitDegrees=5, swing2LimitDegrees=5, twistLimitDegrees=5}
 {bone1Name=dyn_hairstrand_bt_01_l, bone2Name=dyn_hairstrand_bt_01_l_MSTR, swing1LimitDegrees=5, swing2LimitDegrees=5, twistLimitDegrees=5}
 {bone1Name=dyn_hairstrand_tp_01_l, bone2Name=dyn_hairstrand_tp_01_l_MSTR, swing1LimitDegrees=5, swing2LimitDegrees=5, twistLimitDegrees=5}
 {bone1Name=dyn_hairstrand_tp_01, bone2Name=dyn_hairstrand_tp_01_MSTR, swing1LimitDegrees=5, swing2LimitDegrees=5, twistLimitDegrees=5}
 {bone1Name=dyn_hoodie_bt_01_l, bone2Name=dyn_hoodie_bt_01_l_MSTR, swing1LimitDegrees=5, swing2LimitDegrees=10, twistLimitDegrees=5}
 {bone1Name=dyn_hoodie_bt_01, bone2Name=dyn_hoodie_bt_01_MSTR, swing1LimitDegrees=15, swing2LimitDegrees=15, twistLimitDegrees=5}
 {bone1Name=dyn_hoodie_bt_01_r, bone2Name=dyn_hoodie_bt_01_r_MSTR, swing1LimitDegrees=5, swing2LimitDegrees=10, twistLimitDegrees=5}
 {bone1Name=dyn_hoodie_bk_01_l, bone2Name=dyn_hoodie_bk_01_l_MSTR, swing1LimitDegrees=5, swing2LimitDegrees=25, twistLimitDegrees=5}
 {
  bone1Name=dyn_hair_bk_01_l_MSTR
  bone2Name=head
  swing1LimitDegrees=45
  swing1Motion=Free
  swing2LimitDegrees=45
  swing2Motion=Free
  twistLimitDegrees=45
  twistMotion=Free
 }
 {
  bone1Name=dyn_hair_bk_01_MSTR
  bone2Name=head
  swing1LimitDegrees=45
  swing1Motion=Free
  swing2LimitDegrees=45
  swing2Motion=Free
  twistLimitDegrees=45
  twistMotion=Free
 }
 {
  bone1Name=dyn_hair_bk_01_r_MSTR
  bone2Name=head
  swing1LimitDegrees=45
  swing1Motion=Free
  swing2LimitDegrees=45
  swing2Motion=Free
  twistLimitDegrees=45
  twistMotion=Free
 }
]
