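# Post-process style AnimBP for the female Hoodie Bandit face accessory, adding secondary motion to hair and hood atop the attached parent's copied pose.
# Head-to-spine rotation weights back-hair Look Ats toward offset aim bones, while constraints tie hood bones to clavicle rotation.
# Separate RigidBody sims then move hood and hair, their gravity replaced by scaled skydiving force when skydiving.
# A third, root-space RigidBody node is disconnected from the output.
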AnimBlueprint /BRCosmetics/Characters/Player/Female/Medium/Bodies/F_MED_Hoodie_Bandit/Meshes/Parts/F_MED_Hoodie_Bandit_FaceAcc_V2_AnimBP
target_skeleton=&/F_MED_Hoodie_Bandit_FaceAcc_Skeleton
graphs [
 AnimGraph {nodes [
   n1 = Get windVector() -> (windVector:Vector)
   RigidBody!(ComponentPose, Alpha, OverrideWorldGravity:Vector=@n1, ExternalForce) -> (Pose) {
    bClampLinearTranslationLimitToRefPose=true
    bOverrideWorldGravity=true
    overridePhysicsAsset=&/F_MED_Hoodie_Bandit_FaceAcc_V2_Physics
    simulationSpace=BaseBoneSpace
    alphaBoolBlend {blendOption=Linear}
    baseBoneRef {boneName=root}
    componentAppliedLinearAccClamp=<3000 3000 3000>
    componentLinearAccScale=<0.5 0.5 0.5>
    componentLinearVelScale=<0.7 0.7 0.7>
    overrideWorldGravity {z=-980}
   }
   n2 = "Transform (Modify) Bone - Bone: dyn_hair_bk_02_l_aim"(
    ComponentPose=Local To Component(LocalPose=Copy Pose From Mesh() {bUseAttachedParent=true})
    Alpha=1
    Translation="0,-2.5,0"
    Rotation="0,0,0"
    Scale="1,1,1"
   ) {
    translationMode=BMM_Additive
    translationSpace=BCS_BoneSpace
    alphaBoolBlend {blendOption=Linear}
    boneToModify {boneName=dyn_hair_bk_02_l_aim}
    translation {y=-2.5}
   }
   n3 = "Transform (Modify) Bone - Bone: dyn_hair_bk_02_r_aim"(ComponentPose=@n2, Alpha=1, Translation="0,2.5,0", Rotation="0,0,0", Scale="1,1,1") {
    translationMode=BMM_Additive
    translationSpace=BCS_BoneSpace
    alphaBoolBlend {blendOption=Linear}
    boneToModify {boneName=dyn_hair_bk_02_r_aim}
    translation {y=2.5}
   }
   n4 = "Transform (Modify) Bone - Bone: dyn_hair_bk_02_aim"(ComponentPose=@n3, Alpha=1, Translation="5,-0.000,0", Rotation="0,0,0", Scale="1,1,1") {
    translationMode=BMM_Additive
    translationSpace=BCS_BoneSpace
    alphaBoolBlend {blendOption=Linear}
    boneToModify {boneName=dyn_hair_bk_02_aim}
    translation {x=5}
   }
   n5 = Get headAlpha() -> (headAlpha:double)
   n6 = Look At - Bone: dyn_hair_bk_01(ComponentPose=@n4, Alpha=@n5) {
    alphaBoolBlend {blendOption=Linear}
    boneToModify {boneName=dyn_hair_bk_01}
    lookAtLocation {x=0}
    lookAtTarget {boneReference {boneName=dyn_hair_bk_02_aim}}
    lookAt_Axis {axis=<1 0>}
    lookUp_Axis {bInLocalSpace=false, axis {x=0, z=1}}
   }
   n7 = Look At - Bone: dyn_hair_bk_01_r(ComponentPose=@n6, Alpha=@n5) {
    alphaBoolBlend {blendOption=Linear}
    boneToModify {boneName=dyn_hair_bk_01_r}
    lookAtLocation {x=0}
    lookAtTarget {boneReference {boneName=dyn_hair_bk_02_r_aim}}
    lookAt_Axis {axis=<1 0>}
    lookUp_Axis {bInLocalSpace=false, axis {x=0, z=1}}
   }
   n8 = Look At - Bone: dyn_hair_bk_01_l(ComponentPose=@n7, Alpha=@n5) {
    alphaBoolBlend {blendOption=Linear}
    boneToModify {boneName=dyn_hair_bk_01_l}
    lookAtLocation {x=0}
    lookAtTarget {boneReference {boneName=dyn_hair_bk_02_l_aim}}
    lookAt_Axis {axis=<1 0>}
    lookUp_Axis {bInLocalSpace=false, axis {x=0, z=1}}
   }
   n9 = "Constraint - dyn_hoodie_bt_01_l_MSTR (R)"(ComponentPose=@n8, Alpha=1, ConstraintWeights_0=1) {
    alphaBoolBlend {blendOption=Linear}
    boneToModify {boneName=dyn_hoodie_bt_01_l_MSTR}
    constraintSetup [{transformType=Rotation, targetBone {boneName=clavicle_l}}]
    constraintWeights [1]
   }
   n10 = "Constraint - dyn_hoodie_bt_01_r_MSTR (R)"(ComponentPose=@n9, Alpha=1, ConstraintWeights_0=1) {
    alphaBoolBlend {blendOption=Linear}
    boneToModify {boneName=dyn_hoodie_bt_01_r_MSTR}
    constraintSetup [{transformType=Rotation, targetBone {boneName=clavicle_r}}]
    constraintWeights [1]
   }
   n11 = RigidBody(ComponentPose=@n10, OverrideWorldGravity=@n1) {
    bClampLinearTranslationLimitToRefPose=true
    bOverrideWorldGravity=true
    overridePhysicsAsset=&/F_MED_Hoodie_Bandit_FaceAcc_V2_Physics_hood
    simulationSpace=BaseBoneSpace
    alphaBoolBlend {blendOption=Linear}
    baseBoneRef {boneName=spine_05}
    componentAppliedLinearAccClamp=<2000 2000 2000>
    componentLinearAccScale=<0.3 0.3 0.4>
    componentLinearVelScale=<0.3 0.3 0.4>
    overrideWorldGravity {z=-980}
   }
   n12 = RigidBody(ComponentPose=@n11, OverrideWorldGravity=@n1) {
    bClampLinearTranslationLimitToRefPose=true
    bOverrideWorldGravity=true
    overridePhysicsAsset=&/F_MED_Hoodie_Bandit_FaceAcc_V2_Physics_hair
    simulationSpace=BaseBoneSpace
    alphaBoolBlend {blendOption=Linear}
    baseBoneRef {boneName=head}
    componentAppliedLinearAccClamp=<2000 2000 2000>
    componentLinearAccScale=<0.3 0.3 0.35>
    componentLinearVelScale=<0.3 0.3 0.35>
    overrideWorldGravity {z=-980}
   }
   Output Pose(Result=Component To Local(ComponentPose=@n12)) {layerGroup=None}
  ]}
 EventGraph {nodes [
   Event BlueprintUpdateAnimation()
   > n1 = Sequence()
   n2 = Break Fort Anim Input Skydiving External Force(FortAnimInput_SkydivingExternalForce=Get SkydivingExternalForceInput()) -> (
    bUseSkydivingVectorForce=~
    SkydivingVectorForce=~
    HeadToPelvisDirection=~
    FloatingMultiplier=~
    FloatingAdditive=~
    DivingMultiplier=~
    DivingAdditive=~
    ParachutingMultiplier=~
    ParachutingAdditive=~
   )
   > Set windVector(windVector:Vector=SelectVector(A=vector * float(A=@n2.SkydivingVectorForce, B=0.15), B="0, 0,-980", bPickA=Get bIsSkyDiving())) -> (Output_Get:Vector)
   Set headAlpha(
    execute=@n1.then_1
    headAlpha:double=GetAlphaFromReferenceTransformDelta(
     SkeletalMeshComponent=GetOwnerSkeletalMeshComponent()
     BoneName=head
     BaseName=spine_05
     InputRange=(X=-10,Y=10)
     OutputRange=(X=1,Y=0)
     TwistAxis="1,0,0"
    )
   ) -> (Output_Get:double)
   TryGetPawnOwner()
  ]}
]
variables [
 headAlpha {type=number}
 windVector {type=Vector}
]
